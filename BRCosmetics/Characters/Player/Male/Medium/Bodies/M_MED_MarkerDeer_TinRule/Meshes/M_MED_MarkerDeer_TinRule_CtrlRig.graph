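# Control Rig for the MarkerDeer TinRule medium male cosmetic body, which procedurally corrects its belt, bandolier, bag and scarf as the character moves.
# Spherical pose readers and delta-rotation functions measure thigh, head, left upper arm and spine motion, producing corrective weights.
# These combine with state variables like Crouching, DBNO, Skydiving, Melee and Selected_Pistol to blend additive bone offsets.
# Weighted parent constraints position bandolier nulls on the body, and stretchy B-spline IK shapes front and rear strap chains.
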ControlRigBlueprint /BRCosmetics/Characters/Player/Male/Medium/Bodies/M_MED_MarkerDeer_TinRule/Meshes/M_MED_MarkerDeer_TinRule_CtrlRig
graphs [
 Multiply_2_ContainedGraph {nodes [
   n1 = RigVMFunctionEntryNode()
   Multiply = &RigVMFunction_MathDoubleMul(A=@n1.A, B=@n1.B)
   RigVMFunctionReturnNode(Result=&RigVMFunction_MathDoubleMul(B=@n1.C, A=@Multiply.Result).Result)
  ]}
 Multiply_3_ContainedGraph {nodes [
   n1 = RigVMFunctionEntryNode()
   RigVMFunctionReturnNode(Result=&RigVMFunction_MathDoubleMul(B=@n1.C, A=&RigVMFunction_MathDoubleMul(A=@n1.A, B=@n1.B).Result).Result)
  ]}
 Multiply_4_ContainedGraph {nodes [
   n1 = RigVMFunctionEntryNode()
   RigVMFunctionReturnNode(Result=&RigVMFunction_MathDoubleMul(B=@n1.C, A=&RigVMFunction_MathDoubleMul(A=@n1.A, B=@n1.B).Result).Result)
  ]}
 Multiply_4_ContainedGraph {nodes [
   n1 = RigVMFunctionEntryNode()
   RigVMFunctionReturnNode(Result=&RigVMFunction_MathDoubleMul(B=@n1.C, A=&RigVMFunction_MathDoubleMul(A=@n1.A, B=@n1.B).Result).Result)
  ]}
 Multiply_7_ContainedGraph {nodes [
   n1 = RigVMFunctionEntryNode()
   RigVMFunctionReturnNode(Result=&RigVMFunction_MathDoubleMul(B=@n1.C, A=&RigVMFunction_MathDoubleMul(A=@n1.A, B=@n1.B).Result).Result)
  ]}
 New Function_ContainedGraph {nodes [
   RigVMFunctionEntryNode()
   > RigVMFunction_Sequence = RigVMAggregateNode()
   n1 = &RigUnit_SphericalPoseReader(
    ActiveRegionScaleFactors=(PositiveWidth=4,NegativeWidth=1,PositiveHeight=4,NegativeHeight=4)
    ActiveRegionSize=0.1
    Debug=(bDrawDebug=false,DebugScale=25,DebugSegments=20,DebugThickness=0.25)
    DriverAxis=(X=-1,Y=0,Z=0)
    DriverItem=(Type=Bone,Name="thigh_l")
    FalloffRegionScaleFactors=(PositiveWidth=1,NegativeWidth=1.3,PositiveHeight=1,NegativeHeight=1)
    FalloffSize=0.48
    OptionalParentItem=(Type=Bone,Name="None")
    RotationOffset=(X=90,Y=0,Z=-17)
    ExecutePin=@RigVMFunction_Sequence.A
   )
   RigVMFunction_AnimEvalRichCurve = &RigVMFunction_AnimEvalRichCurve(
    Curve=(EditorCurveData=(Keys=((InterpMode=RCIM_Cubic),(InterpMode=RCIM_Cubic,TangentMode=RCTM_User,Time=0.6812,Value=0.3663,ArriveTangent=1.5267,LeaveTangent=1.5267),(Time=1,Value=1))))
    SourceMinimum=0
    TargetMinimum=0
    Value=@n1.OutputParam
   )
   > RigVMVariableNode(Value=@RigVMFunction_AnimEvalRichCurve.Result) {variable=Front_Leg_L}
   > n2 = &RigUnit_SphericalPoseReader(
    ActiveRegionScaleFactors=(PositiveWidth=3,NegativeWidth=2,PositiveHeight=1,NegativeHeight=1)
    ActiveRegionSize=0.1
    Debug=(bDrawDebug=false,DebugScale=25,DebugSegments=20,DebugThickness=0.25)
    DriverAxis=(X=-1,Y=0,Z=0)
    DriverItem=(Type=Bone,Name="thigh_l")
    FalloffRegionScaleFactors=(PositiveWidth=1,NegativeWidth=1,PositiveHeight=1,NegativeHeight=1.5)
    FalloffSize=0.45
    FlipHeightScaling=false
    FlipWidthScaling=false
    OptionalParentItem=(Type=Bone,Name="None")
    RotationOffset=(X=200,Y=20,Z=80)
   )
   > RigVMVariableNode(Value=@n2.OutputParam) {variable=Side_Leg_L}
   > n3 = &RigUnit_SphericalPoseReader(
    ActiveRegionScaleFactors=(PositiveWidth=3.5,NegativeWidth=0.5,PositiveHeight=3,NegativeHeight=3)
    ActiveRegionSize=0.1
    Debug=(bDrawDebug=false,DebugScale=25,DebugSegments=20,DebugThickness=0.25)
    DriverAxis=(X=-1,Y=0,Z=0)
    DriverItem=(Type=Bone,Name="thigh_l")
    FalloffRegionScaleFactors=(PositiveWidth=1,NegativeWidth=1,PositiveHeight=1,NegativeHeight=1)
    FalloffSize=0.48
    OptionalParentItem=(Type=Bone,Name="None")
    RotationOffset=(X=-90,Y=0,Z=0)
   )
   > RigVMVariableNode(Value=@n3.OutputParam) {variable=Back_leg_L}
   > n4 = &RigUnit_SphericalPoseReader(
    ActiveRegionScaleFactors=(PositiveWidth=4,NegativeWidth=1,PositiveHeight=4,NegativeHeight=4)
    ActiveRegionSize=0.1
    Debug=(bDrawDebug=false,DebugScale=25,DebugSegments=20,DebugThickness=0.25)
    DriverAxis=(X=1,Y=0,Z=0)
    DriverItem=(Type=Bone,Name="thigh_r")
    FalloffRegionScaleFactors=(PositiveWidth=1,NegativeWidth=1.3,PositiveHeight=1,NegativeHeight=1)
    FalloffSize=0.48
    FlipHeightScaling=false
    FlipWidthScaling=true
    OptionalParentItem=(Type=Bone,Name="None")
    RotationOffset=(X=-90,Y=0,Z=-17)
   )
   > RigVMVariableNode(Value=@n4.OutputParam) {variable=Front_Leg_R}
   > n5 = &RigUnit_SphericalPoseReader(
    ActiveRegionScaleFactors=(PositiveWidth=2.5,NegativeWidth=2,PositiveHeight=1,NegativeHeight=1.3)
    ActiveRegionSize=0.1
    Debug=(bDrawDebug=false,bDraw2D=false,DebugScale=25,DebugSegments=20,DebugThickness=0.25)
    DriverAxis=(X=1,Y=0,Z=0)
    DriverItem=(Type=Bone,Name="thigh_r")
    FalloffRegionScaleFactors=(PositiveWidth=1,NegativeWidth=1,PositiveHeight=1,NegativeHeight=2.2)
    FalloffSize=0.35
    FlipHeightScaling=false
    FlipWidthScaling=true
    OptionalParentItem=(Type=Bone,Name="None")
    RotationOffset=(X=20,Y=20,Z=90)
   )
   > RigVMVariableNode(Value=@n5.OutputParam) {variable=Side_Leg_R}
   > n6 = &RigUnit_SphericalPoseReader(
    ActiveRegionScaleFactors=(PositiveWidth=3.5,NegativeWidth=0.5,PositiveHeight=3,NegativeHeight=3)
    ActiveRegionSize=0.1
    Debug=(bDrawDebug=false,DebugScale=25,DebugSegments=20,DebugThickness=0.25)
    DriverAxis=(X=-1,Y=0,Z=0)
    DriverItem=(Type=Bone,Name="thigh_twist_01_r")
    FalloffRegionScaleFactors=(PositiveWidth=1,NegativeWidth=1,PositiveHeight=1,NegativeHeight=1)
    FalloffSize=0.48
    OptionalParentItem=(Type=Bone,Name="pelvis")
    RotationOffset=(X=-90,Y=0,Z=0)
   )
   > RigVMVariableNode(Value=@n6.OutputParam) {variable=Back_Leg_R}
   > DeltaRotation = RigVMFunctionReferenceNode(ParentType=Null, child=spine_05, max=60, min=47, onlyTwist=true, parent=spine_05_null, primAxis=0, twistAxis=(X=1,Y=0,Z=0))
   > &RigUnit_ModifyTransforms(
    ItemToModify=((Item=(Type=Bone,Name="dyn_fr_belt_extra_l"),Transform=(Rotation=(X=0,Y=0,Z=0,W=1),Translation=(X=0,Y=0,Z=2.5),Scale3D=(X=1,Y=1,Z=1))))
    WeightMinimum=0
    Weight=&RigVMFunction_MathDoubleMul(A=@DeltaRotation.ouputWeight, B=RigVMVariableNode() {variable=Crouching}.Value).Result
   )
   > DeltaRotation_1 = RigVMFunctionReferenceNode(ParentType=Null, child=spine_05, max=50, min=20, onlyTwist=true, parent=spine_05_null, primAxis=0, twistAxis=(X=1,Y=0,Z=0))
   Remap = &RigVMFunction_MathDoubleRemap(SourceMaximum=1, SourceMinimum=0, TargetMaximum=0, TargetMinimum=1, Value=RigVMVariableNode() {variable=Selected_Pistol}.Value)
   Multiply_6 = RigVMAggregateNode(A=@DeltaRotation_1.ouputWeight, C=@Remap.Result, B=RigVMVariableNode() {variable=Crouching}.Value)
   > &RigUnit_ModifyTransforms(
    ItemToModify=((Item=(Type=Bone,Name="dyn_fr_belt_extra_l"),Transform=(Rotation=(X=-0.0436,Y=0,Z=0,W=0.999),Translation=(X=0,Y=-1,Z=1.5),Scale3D=(X=1,Y=1,Z=1))))
    WeightMinimum=0
    Weight=@Multiply_6.Result
   )
   &RigUnit_ModifyTransforms(
    ItemToModify="((Item=(Type=Bone,Name=\"base_fr_belt_l\"),Transform=(Rotation=(X=0,Y=0.1426,Z=0,W=-0.9898),Translation=(X=3.9427,Y=-0.3445,Z=6.1861),Scale3D=(X=1,Y=1,Z=1))),(Item=(Type=Bone,Name=\"base_sd_belt_l\"),Transform=(Rotation=(X=0.1207,Y=0,Z=0,W=0.9927),Translation= ...(truncated)"
    WeightMinimum=0
    ExecutePin=@RigVMFunction_Sequence.B
    Weight=RigVMVariableNode() {variable=Front_Leg_L}.Value
   )
   > &RigUnit_ModifyTransforms(
    ItemToModify="((Item=(Type=Bone,Name=\"base_fr_belt_l\"),Transform=(Rotation=(X=0.0466,Y=0,Z=0,W=-0.9989),Translation=(X=0.2648,Y=-1.7418,Z=1.2603),Scale3D=(X=1,Y=1,Z=1))),(Item=(Type=Bone,Name=\"base_sd_belt_l\"),Transform=(Rotation=(X=-0.0343,Y=0.0145,Z=0.013,W=-0.9992),T ...(truncated)"
    WeightMinimum=0
    Weight=RigVMVariableNode() {variable=Side_Leg_L}.Value
   )
   > &RigUnit_ModifyTransforms(
    ItemToModify="((Item=(Type=Bone,Name=\"base_fr_belt_l\"),Transform=(Rotation=(X=0,Y=0,Z=0,W=1),Translation=(X=0,Y=0,Z=0),Scale3D=(X=1,Y=1,Z=1))),(Item=(Type=Bone,Name=\"base_sd_belt_l\"),Transform=(Rotation=(X=0,Y=0,Z=0,W=1),Translation=(X=0,Y=0,Z=0),Scale3D=(X=1,Y=1,Z=1))) ...(truncated)"
    WeightMinimum=0
    Weight=RigVMVariableNode() {variable=Back_leg_L}.Value
   )
   > &RigUnit_ModifyTransforms(
    ItemToModify="((Item=(Type=Bone,Name=\"base_fr_belt_r\"),Transform=(Rotation=(X=-0.0259,Y=0.021,Z=0.1185,W=-0.9924),Translation=(X=1.1999,Y=1.6118,Z=10.6362),Scale3D=(X=1,Y=1,Z=1))),(Item=(Type=Bone,Name=\"base_sd_belt_r\"),Transform=(Rotation=(X=0.176,Y=0.0034,Z=0.0076,W=- ...(truncated)"
    WeightMinimum=0
    Weight=RigVMVariableNode() {variable=Front_Leg_R}.Value
   )
   > &RigUnit_ModifyTransforms(
    ItemToModify="((Item=(Type=Bone,Name=\"base_fr_belt_r\"),Transform=(Rotation=(X=0.1921,Y=0,Z=0,W=0.9814),Translation=(X=0,Y=-0.0304,Z=4.2526),Scale3D=(X=1,Y=1,Z=1))),(Item=(Type=Bone,Name=\"base_sd_belt_r\"),Transform=(Rotation=(X=-0.0088,Y=0.0969,Z=0.0904,W=-0.9911),Transl ...(truncated)"
    WeightMinimum=0
    Weight=RigVMVariableNode() {variable=Side_Leg_R}.Value
   )
   > &RigUnit_ModifyTransforms(
    ItemToModify="((Item=(Type=Bone,Name=\"base_fr_belt_r\"),Transform=(Rotation=(X=0,Y=0,Z=0,W=1),Translation=(X=0,Y=0,Z=0),Scale3D=(X=1,Y=1,Z=1))),(Item=(Type=Bone,Name=\"base_sd_belt_r\"),Transform=(Rotation=(X=-0.0175,Y=0,Z=0,W=0.9998),Translation=(X=0,Y=-0.0268,Z=2.5691),S ...(truncated)"
    WeightMinimum=0
    Weight=RigVMVariableNode() {variable=Back_Leg_R}.Value
   )
   &RigUnit_ModifyTransforms(
    ItemToModify="((Item=(Type=Bone,Name=\"base_fr_belt_l\"),Transform=(Rotation=(X=0,Y=0,Z=0,W=1),Translation=(X=-3,Y=0,Z=0),Scale3D=(X=1,Y=1,Z=1))),(Item=(Type=Bone,Name=\"base_sd_belt_l\"),Transform=(Rotation=(X=0,Y=0.1736,Z=0,W=0.9848),Translation=(X=-2,Y=-1,Z=-2),Scale3D=( ...(truncated)"
    WeightMinimum=0
    Weight=&RigVMFunction_MathDoubleMul(A=RigVMVariableNode() {variable=Front_Leg_L}.Value, B=RigVMVariableNode() {variable=DBNO}.Value).Result
    ExecutePin=@RigVMFunction_Sequence.E
   )
   > &RigUnit_ModifyTransforms(
    ItemToModify="((Item=(Type=Bone,Name=\"base_fr_belt_l\"),Transform=(Rotation=(X=0,Y=0,Z=0,W=1),Translation=(X=0,Y=0,Z=0),Scale3D=(X=1,Y=1,Z=1))),(Item=(Type=Bone,Name=\"base_sd_belt_l\"),Transform=(Rotation=(X=0,Y=0,Z=0,W=1),Translation=(X=0,Y=0,Z=0),Scale3D=(X=1,Y=1,Z=1))) ...(truncated)"
    WeightMinimum=0
    Weight=&RigVMFunction_MathDoubleMul(A=RigVMVariableNode() {variable=Side_Leg_L}.Value, B=RigVMVariableNode() {variable=DBNO}.Value).Result
   )
   RigVMFunctionReturnNode(ExecuteContext=@RigVMFunction_Sequence.F)
   &RigUnit_ModifyTransforms(
    ItemToModify="((Item=(Type=Bone,Name=\"base_fr_belt_l\"),Transform=(Rotation=(X=-0.1736,Y=0,Z=0,W=0.9848),Translation=(X=-1.8,Y=-2,Z=-2),Scale3D=(X=1,Y=1,Z=1))),(Item=(Type=Bone,Name=\"base_sd_belt_l\"),Transform=(Rotation=(X=0,Y=0.1305,Z=0,W=0.9914),Translation=(X=-2.5,Y=0 ...(truncated)"
    WeightMinimum=0
    Weight=&RigVMFunction_MathDoubleMul(A=RigVMVariableNode() {variable=Front_Leg_L}.Value, B=RigVMVariableNode() {variable=Crouching}.Value).Result
    ExecutePin=@RigVMFunction_Sequence.D
   )
   > &RigUnit_ModifyTransforms(
    ItemToModify="((Item=(Type=Bone,Name=\"base_fr_belt_l\"),Transform=(Rotation=(X=0,Y=0,Z=0,W=1),Translation=(X=0,Y=0,Z=0),Scale3D=(X=1,Y=1,Z=1))),(Item=(Type=Bone,Name=\"base_sd_belt_l\"),Transform=(Rotation=(X=0,Y=0,Z=0,W=1),Translation=(X=0,Y=0,Z=0),Scale3D=(X=1,Y=1,Z=1))) ...(truncated)"
    WeightMinimum=0
    Weight=&RigVMFunction_MathDoubleMul(A=RigVMVariableNode() {variable=Side_Leg_L}.Value, B=RigVMVariableNode() {variable=Crouching}.Value).Result
   )
   > &RigUnit_ModifyTransforms(
    ItemToModify="((Item=(Type=Bone,Name=\"base_fr_belt_l\"),Transform=(Rotation=(X=0,Y=0,Z=0,W=1),Translation=(X=0,Y=0,Z=0),Scale3D=(X=1,Y=1,Z=1))),(Item=(Type=Bone,Name=\"base_sd_belt_l\"),Transform=(Rotation=(X=0,Y=0,Z=0,W=1),Translation=(X=0,Y=0,Z=0),Scale3D=(X=1,Y=1,Z=1))) ...(truncated)"
    WeightMinimum=0
    Weight=&RigVMFunction_MathDoubleMul(A=RigVMVariableNode() {variable=Back_leg_L}.Value, B=RigVMVariableNode() {variable=Crouching}.Value).Result
   )
   > &RigUnit_ModifyTransforms(
    ItemToModify="((Item=(Type=Bone,Name=\"base_sd_belt_r\"),Transform=(Rotation=(X=0,Y=0,Z=0,W=1),Translation=(X=0,Y=0,Z=0),Scale3D=(X=1,Y=1,Z=1))),(Item=(Type=Bone,Name=\"base_bk_belt_r\"),Transform=(Rotation=(X=0,Y=0,Z=0,W=1),Translation=(X=0,Y=0,Z=2),Scale3D=(X=1,Y=1,Z=1))) ...(truncated)"
    WeightMinimum=0
    Weight=&RigVMFunction_MathDoubleMul(B=RigVMVariableNode() {variable=Crouching}.Value, A=RigVMVariableNode() {variable=Front_Leg_R}.Value).Result
   )
   > &RigUnit_ModifyTransforms(
    ItemToModify="((Item=(Type=Bone,Name=\"base_sd_belt_r\"),Transform=(Rotation=(X=0,Y=0,Z=0.0872,W=0.9962),Translation=(X=0,Y=0,Z=0),Scale3D=(X=1,Y=1,Z=1))),(Item=(Type=Bone,Name=\"base_bk_belt_r\"),Transform=(Rotation=(X=-0.0872,Y=0,Z=0,W=0.9962),Translation=(X=-1,Y=0,Z=1),S ...(truncated)"
    WeightMinimum=0
    Weight=&RigVMFunction_MathDoubleMul(B=RigVMVariableNode() {variable=Crouching}.Value, A=RigVMVariableNode() {variable=Side_Leg_R}.Value).Result
   )
   &RigUnit_ModifyTransforms(
    ItemToModify="((Item=(Type=Bone,Name=\"base_fr_belt_l\"),Transform=(Rotation=(X=0,Y=0,Z=0,W=1),Translation=(X=0,Y=0,Z=0),Scale3D=(X=1,Y=1,Z=1))),(Item=(Type=Bone,Name=\"base_sd_belt_l\"),Transform=(Rotation=(X=0,Y=0,Z=0,W=1),Translation=(X=-4.1225,Y=0.0675,Z=0.9489),Scale3D ...(truncated)"
    WeightMinimum=0
    Weight=&RigVMFunction_MathDoubleMul(A=RigVMVariableNode() {variable=Front_Leg_L}.Value, B=RigVMVariableNode() {variable=Side_Leg_L}.Value).Result
    ExecutePin=@RigVMFunction_Sequence.C
   )
   Multiply = &RigVMFunction_MathDoubleMul(A=RigVMVariableNode() {variable=Front_Leg_R}.Value, B=RigVMVariableNode() {variable=Side_Leg_R}.Value)
   > &RigUnit_ModifyTransforms(
    ItemToModify="((Item=(Type=Bone,Name=\"base_fr_belt_r\"),Transform=(Rotation=(X=0,Y=-0.1483,Z=0,W=-0.9889),Translation=(X=0.3084,Y=-1.1475,Z=-3.3182),Scale3D=(X=1,Y=1,Z=1))),(Item=(Type=Bone,Name=\"base_sd_belt_r\"),Transform=(Rotation=(X=0,Y=0,Z=0.0253,W=0.9997),Translatio ...(truncated)"
    WeightMinimum=0
    Weight=@Multiply.Result
   )
  ]}
 New Function_ContainedGraph {nodes [
   n1 = RigVMFunctionEntryNode()
   RigVMFunction_ControlFlowBranch = &RigVMFunction_ControlFlowBranch(Condition=@n1.ControlsDebug, ExecuteContext=@n1.ExecuteContext)
   RigVMFunctionReturnNode(ExecuteContext=@RigVMFunction_ControlFlowBranch.Completed)
   RigUnit_GetTransform = &RigUnit_GetTransform(Item=(Type=Control,Name="thigh_l_ctrl"), bInitial=False)
   RigUnit_SetTransform = &RigUnit_SetTransform(
    Item=(Type=Bone,Name="thigh_l")
    Weight=1
    bInitial=False
    Value=@RigUnit_GetTransform.Transform
    ExecutePin=@RigVMFunction_ControlFlowBranch.True
   )
   > &RigUnit_SetTransform(
    Item=(Type=Bone,Name="thigh_r")
    Weight=1
    bInitial=False
    Value=&RigUnit_GetTransform(Item=(Type=Control,Name="thigh_r_ctrl"), bInitial=False).Transform
   )
   > &RigUnit_SetTransform(
    Item=(Type=Bone,Name="upperarm_l")
    Weight=1
    bInitial=False
    Value=&RigUnit_GetTransform(Item=(Type=Control,Name="upperarm_l_ctrl"), bInitial=False).Transform
   )
   > &RigUnit_SetTransform(
    Item=(Type=Bone,Name="upperarm_r")
    Weight=1
    bInitial=False
    Value=&RigUnit_GetTransform(Item=(Type=Control,Name="upperarm_r_ctrl"), bInitial=False).Transform
   )
   > &RigUnit_SetTransform(
    Item=(Type=Bone,Name="clavicle_l")
    Weight=1
    bInitial=False
    Value=&RigUnit_GetTransform(Item=(Type=Control,Name="clavicle_l_ctrl"), bInitial=False).Transform
   )
   > &RigUnit_SetTransform(
    Item=(Type=Bone,Name="clavicle_r")
    Weight=1
    bInitial=False
    Value=&RigUnit_GetTransform(Item=(Type=Control,Name="clavicle_r_ctrl"), bInitial=False).Transform
   )
   > &RigUnit_SetTransform(
    Item=(Type=Bone,Name="spine_05")
    Weight=1
    bInitial=False
    Value=&RigUnit_GetTransform(Item=(Type=Control,Name="spine_05_ctrl"), bInitial=False).Transform
   )
  ]}
 New Function_ContainedGraph {nodes [
   n1 = RigVMFunctionEntryNode()
   RigVMFunction_Sequence = &RigVMFunction_Sequence(ExecuteContext=@n1.ExecuteContext)
   &RigUnit_ParentConstraint(
    AdvancedSettings=(RotationOrderForFilter=XZY)
    Child=(Type=Bone,Name="def_bandolier_root")
    Parents="((Item=(Type=Bone,Name=\"spine_04\"),Weight=0.0097),(Item=(Type=Bone,Name=\"deform_upperarm_fr_r\"),Weight=0.0153),(Item=(Type=Bone,Name=\"spine_03\"),Weight=0.0188),(Item=(Type=Bone,Name=\"neck_02\"),Weight=0.0216),(Item=(Type=Bone,Name=\"upperarm_twist_01_r\"),Wei ...(truncated)"
    Weight=1
    ExecutePin=@RigVMFunction_Sequence.A
   )
   > &RigUnit_ParentConstraint(
    AdvancedSettings=(RotationOrderForFilter=XZY)
    Child=(Type=Null,Name="def_bandolier_front_01_null")
    Parents="((Item=(Type=Bone,Name=\"neck_02\"),Weight=0.0178),(Item=(Type=Bone,Name=\"deform_upperarm_up_r\"),Weight=0.0245),(Item=(Type=Bone,Name=\"upperarm_twist_01_r\"),Weight=0.028),(Item=(Type=Bone,Name=\"spine_04\"),Weight=0.0328),(Item=(Type=Bone,Name=\"deform_pec_r\"), ...(truncated)"
    Weight=1
   )
   > &RigUnit_ParentConstraint(
    AdvancedSettings=(RotationOrderForFilter=XZY)
    Child=(Type=Null,Name="def_bandolier_front_02_null")
    Parents="((Item=(Type=Bone,Name=\"neck_02\"),Weight=0.0005),(Item=(Type=Bone,Name=\"spine_02\"),Weight=0.0055),(Item=(Type=Bone,Name=\"deform_upperarm_fr_r\"),Weight=0.0266),(Item=(Type=Bone,Name=\"spine_03\"),Weight=0.055),(Item=(Type=Bone,Name=\"deform_pec_r\"),Weight=0.11 ...(truncated)"
    Weight=1
   )
   > &RigUnit_ParentConstraint(
    AdvancedSettings=(RotationOrderForFilter=XZY)
    Child=(Type=Null,Name="def_bandolier_front_03_null")
    Parents="((Item=(Type=Bone,Name=\"deform_upperarm_fr_r\"),Weight=0.0093),(Item=(Type=Bone,Name=\"spine_01\"),Weight=0.0102),(Item=(Type=Bone,Name=\"spine_02\"),Weight=0.0875),(Item=(Type=Bone,Name=\"clavicle_r\"),Weight=0.102),(Item=(Type=Bone,Name=\"spine_05\"),Weight=0.112 ...(truncated)"
    Weight=1
   )
   > &RigUnit_ParentConstraint(
    AdvancedSettings=(RotationOrderForFilter=XZY)
    Child=(Type=Null,Name="def_bandolier_front_04_null")
    Parents="((Item=(Type=Bone,Name=\"pelvis\"),Weight=0.0001),(Item=(Type=Bone,Name=\"clavicle_r\"),Weight=0.0334),(Item=(Type=Bone,Name=\"deform_pec_r\"),Weight=0.0352),(Item=(Type=Bone,Name=\"spine_05\"),Weight=0.044),(Item=(Type=Bone,Name=\"spine_01\"),Weight=0.0489),(Item=( ...(truncated)"
    Weight=1
   )
   > &RigUnit_ParentConstraint(
    AdvancedSettings=(RotationOrderForFilter=XZY)
    Child=(Type=Null,Name="def_bandolier_rear_01_null")
    Parents="((Item=(Type=Bone,Name=\"deform_upperarm_up_r\"),Weight=0.0038),(Item=(Type=Bone,Name=\"upperarm_twist_01_r\"),Weight=0.0063),(Item=(Type=Bone,Name=\"neck_02\"),Weight=0.008),(Item=(Type=Bone,Name=\"spine_04\"),Weight=0.0085),(Item=(Type=Bone,Name=\"deform_upperarm ...(truncated)"
    Weight=1
   )
   > &RigUnit_ParentConstraint(
    AdvancedSettings=(RotationOrderForFilter=XZY)
    Child=(Type=Null,Name="def_bandolier_rear_02_null")
    Parents="((Item=(Type=Bone,Name=\"upperarm_twist_02_r\"),Weight=0.0002),(Item=(Type=Bone,Name=\"upperarm_twist_01_r\"),Weight=0.0002),(Item=(Type=Bone,Name=\"deform_upperarm_bk_r\"),Weight=0.0243),(Item=(Type=Bone,Name=\"spine_04\"),Weight=0.0638),(Item=(Type=Bone,Name=\"sp ...(truncated)"
    Weight=1
   )
   > &RigUnit_ParentConstraint(
    AdvancedSettings=(RotationOrderForFilter=XZY)
    Child=(Type=Null,Name="def_bandolier_rear_03_null")
    Parents="((Item=(Type=Bone,Name=\"deform_upperarm_bk_r\"),Weight=0.0093),(Item=(Type=Bone,Name=\"spine_04\"),Weight=0.1038),(Item=(Type=Bone,Name=\"spine_05\"),Weight=0.2018),(Item=(Type=Bone,Name=\"spine_03\"),Weight=0.2898),(Item=(Type=Bone,Name=\"clavicle_r\"),Weight=0.39 ...(truncated)"
    Weight=1
   )
   > &RigUnit_ParentConstraint(
    AdvancedSettings=(RotationOrderForFilter=XZY)
    Child=(Type=Null,Name="def_bandolier_rear_04_null")
    Parents="((Item=(Type=Bone,Name=\"spine_01\"),Weight=0.0034),(Item=(Type=Bone,Name=\"spine_02\"),Weight=0.0407),(Item=(Type=Bone,Name=\"clavicle_r\"),Weight=0.1015),(Item=(Type=Bone,Name=\"spine_05\"),Weight=0.2071),(Item=(Type=Bone,Name=\"spine_04\"),Weight=0.2586),(Item=(T ...(truncated)"
    Weight=1
   )
   RigVMFunctionReturnNode(ExecuteContext=@RigVMFunction_Sequence.B)
  ]}
 New Function_ContainedGraph {nodes [
   n1 = RigVMFunctionEntryNode()
   RigVMFunction_Sequence = &RigVMFunction_Sequence(ExecuteContext=@n1.ExecuteContext)
   &RigUnit_ModifyTransforms(
    ItemToModify="((Item=(Type=Null,Name=\"def_bandolier_rear_02_null\"),Transform=(Rotation=(X=0,Y=0,Z=0,W=1),Translation=(X=0,Y=0,Z=3),Scale3D=(X=1,Y=1,Z=1))),(Item=(Type=Null,Name=\"def_bandolier_rear_03_null\"),Transform=(Rotation=(X=0,Y=0,Z=0,W=1),Translation=(X=0,Y=0,Z=0) ...(truncated)"
    Weight=1
    WeightMinimum=0
    ExecutePin=@RigVMFunction_Sequence.A
   )
   > &RigUnit_ModifyTransforms(
    ItemToModify=((Item=(Type=Bone,Name="dyn_bag_chain_01"),Transform=(Rotation=(X=0,Y=-0.342,Z=0,W=0.9397),Translation=(X=0,Y=0,Z=0),Scale3D=(X=1,Y=1,Z=1))))
    WeightMinimum=0
    Weight=&RigVMFunction_MathDoubleMul(A=RigVMVariableNode() {variable=Melee}.Value, B=RigVMVariableNode() {variable=Crouching}.Value).Result
   )
   > DeltaRotationZ = RigVMFunctionReferenceNode(Child=head, Parent=spine_05, SourceMax=40, SourceMin=10)
   Multiply_2 = RigVMAggregateNode(C=@DeltaRotationZ.Delta Z, A=RigVMVariableNode() {variable=Melee}.Value, B=RigVMVariableNode() {variable=Crouching}.Value)
   > &RigUnit_ParentConstraint(
    AdvancedSettings=(InterpolationType=Average,RotationOrderForFilter=XZY)
    Child=(Type=Bone,Name="base_bag")
    Parents=((Item=(Type=Bone,Name="pelvis"),Weight=1))
    Weight=@Multiply_2.Result
   )
   > &RigUnit_ModifyTransforms(
    ItemToModify="((Item=(Type=Null,Name=\"def_bandolier_front_04_null\"),Transform=(Rotation=(X=0,Y=0.0436,Z=0,W=0.999),Translation=(X=0,Y=0,Z=0),Scale3D=(X=1,Y=1,Z=1))),(Item=(Type=Null,Name=\"def_bandolier_rear_04_null\"),Transform=(Rotation=(X=0,Y=-0.0872,Z=0,W=0.9962),Tran ...(truncated)"
    WeightMinimum=0
    Weight=RigVMVariableNode() {variable=Bag_Rot}.Value
   )
   > &RigUnit_ModifyTransforms(
    ItemToModify="((Item=(Type=Null,Name=\"def_bandolier_front_04_null\"),Transform=(Rotation=(X=0,Y=0,Z=0,W=1),Translation=(X=0,Y=0,Z=-1),Scale3D=(X=1,Y=1,Z=1))),(Item=(Type=Bone,Name=\"base_bag\"),Transform=(Rotation=(X=0,Y=0,Z=0,W=1),Translation=(X=-6,Y=-2,Z=0),Scale3D=(X=1, ...(truncated)"
    WeightMinimum=0
    Weight=RigVMVariableNode() {variable=Crouching}.Value
   )
   Multiply = &RigVMFunction_MathDoubleMul(B=RigVMVariableNode() {variable=Moving2D}.Value, A=RigVMVariableNode() {variable=Crouching}.Value)
   > &RigUnit_ModifyTransforms(
    ItemToModify="((Item=(Type=Bone,Name=\"base_bag\"),Transform=(Rotation=(X=0,Y=0,Z=0,W=1),Translation=(X=10,Y=0,Z=0),Scale3D=(X=1,Y=1,Z=1))),(Item=(Type=Null,Name=\"def_bandolier_rear_04_null\"),Transform=(Rotation=(X=0,Y=0,Z=0,W=1),Translation=(X=0,Y=0,Z=0),Scale3D=(X=1,Y=1 ...(truncated)"
    WeightMinimum=0
    Weight=@Multiply.Result
   )
   > &RigUnit_ModifyTransforms(
    ItemToModify="((Item=(Type=Bone,Name=\"dyn_bag\"),Transform=(Rotation=(X=0,Y=0,Z=0,W=1),Translation=(X=0,Y=3.5,Z=0),Scale3D=(X=1,Y=1,Z=1))),(Item=(Type=Bone,Name=\"dyn_bag_chain_01\"),Transform=(Rotation=(X=0,Y=0,Z=-0.0872,W=0.9962),Translation=(X=0,Y=0,Z=0),Scale3D=(X=1,Y= ...(truncated)"
    WeightMinimum=0
    Weight=RigVMVariableNode() {variable=Skydiving}.Value
   )
   > &RigUnit_ModifyTransforms(
    ItemToModify="((Item=(Type=Bone,Name=\"dyn_bag\"),Transform=(Rotation=(X=0.1294,Y=0.017,Z=0.1294,W=0.983),Translation=(X=0,Y=0,Z=0),Scale3D=(X=1,Y=1,Z=1))),(Item=(Type=Bone,Name=\"dyn_bag_chain_01\"),Transform=(Rotation=(X=0.0189,Y=-0.2156,Z=0.0851,W=0.9726),Translation=(X= ...(truncated)"
    WeightMinimum=0
    Weight=RigVMVariableNode() {variable=DBNO}.Value
   )
   > SplineIK = RigVMFunctionReferenceNode(
    Primary Axis=(X=1,Y=0,Z=0)
    Samples Per Segment=6
    Secondary Spline Direction=(X=0,Y=-1,Z=0)
    Spline Mode=BSpline
    Stretch=true
    Up Axis=(X=0,Y=-1,Z=0)
    debug=false
    Bones=RigVMVariableNode() {variable=Bandolier_Front_Joints}.Value
    Controls=RigVMVariableNode() {variable=Bandolier_Front_Ctrls}.Value
   )
   > SplineIK_1 = RigVMFunctionReferenceNode(
    Primary Axis=(X=1,Y=0,Z=0)
    Samples Per Segment=6
    Secondary Spline Direction=(X=0,Y=5,Z=0)
    Spline Mode=BSpline
    Stretch=true
    Up Axis=(X=0,Y=0,Z=0)
    Bones=RigVMVariableNode() {variable=Bandolier_Rear_Joints}.Value
    Controls=RigVMVariableNode() {variable=Bandolier_Rear_Ctrls}.Value
   )
   RigVMFunctionReturnNode(ExecuteContext=@RigVMFunction_Sequence.B)
  ]}
 New Function_ContainedGraph {nodes [
   RigVMFunctionEntryNode()
   > RigVMFunction_Sequence = RigVMAggregateNode()
   n1 = &RigUnit_SphericalPoseReader(
    ActiveRegionScaleFactors=(PositiveWidth=1,NegativeWidth=1,PositiveHeight=2.4,NegativeHeight=1.4)
    ActiveRegionSize=0.06
    Debug=(bDrawDebug=false,bDraw2D=False,bDrawLocalAxes=False,DebugScale=25,DebugSegments=20,DebugThickness=0.25)
    DriverAxis=(X=0,Y=1,Z=0)
    DriverItem=(Type=Bone,Name="head")
    FalloffRegionScaleFactors=(PositiveWidth=1,NegativeWidth=1,PositiveHeight=6,NegativeHeight=1.5)
    FalloffSize=0.2
    FlipHeightScaling=true
    FlipWidthScaling=false
    OptionalParentItem=(Type=Bone,Name="spine_05")
    RotationOffset=(X=50,Y=0,Z=30)
    ExecutePin=@RigVMFunction_Sequence.A
   )
   > RigVMVariableNode(Value=@n1.OutputParam) {variable=Head_Rot_Right}
   > n2 = &RigUnit_SphericalPoseReader(
    ActiveRegionScaleFactors=(PositiveWidth=1,NegativeWidth=1,PositiveHeight=2.4,NegativeHeight=0.5)
    ActiveRegionSize=0.07
    Debug=(bDrawDebug=true,bDraw2D=False,bDrawLocalAxes=False,DebugScale=25,DebugSegments=20,DebugThickness=0.25)
    DriverAxis=(X=0,Y=1,Z=0)
    DriverItem=(Type=Bone,Name="head")
    FalloffRegionScaleFactors=(PositiveWidth=1,NegativeWidth=1,PositiveHeight=6,NegativeHeight=1)
    FalloffSize=0.22
    FlipHeightScaling=False
    FlipWidthScaling=False
    OptionalParentItem=(Type=Bone,Name="spine_05")
    RotationOffset=(X=148,Y=0,Z=37)
   )
   > RigVMVariableNode(Value=@n2.OutputParam) {variable=Head_Rot_Left}
   > n3 = &RigUnit_SphericalPoseReader(
    ActiveRegionScaleFactors=(PositiveWidth=0.2,NegativeWidth=1.5,PositiveHeight=1.5,NegativeHeight=1.5)
    ActiveRegionSize=0.2
    Debug=(bDrawDebug=false,bDraw2D=False,bDrawLocalAxes=False,DebugScale=25,DebugSegments=20,DebugThickness=0.25)
    DriverAxis=(X=1,Y=0,Z=0)
    DriverItem=(Type=Bone,Name="head")
    FalloffRegionScaleFactors=(PositiveWidth=1,NegativeWidth=1,PositiveHeight=2,NegativeHeight=2)
    FalloffSize=0.3
    FlipHeightScaling=False
    FlipWidthScaling=False
    OptionalParentItem=(Type=Bone,Name="spine_05")
    RotationOffset=(X=-90,Y=0,Z=25)
   )
   > RigVMVariableNode(Value=@n3.OutputParam) {variable=Head_Rot_Up}
   > n4 = &RigUnit_SphericalPoseReader(
    ActiveRegionScaleFactors=(PositiveWidth=3,NegativeWidth=3,PositiveHeight=3,NegativeHeight=3)
    ActiveRegionSize=0.08
    Debug=(bDrawDebug=false,DebugScale=25,DebugSegments=20,DebugThickness=0.25)
    DriverAxis=(X=1,Y=0,Z=0)
    DriverItem=(Type=Bone,Name="upperarm_l")
    FalloffRegionScaleFactors=(PositiveWidth=2.1,NegativeWidth=2,PositiveHeight=2,NegativeHeight=2)
    FalloffSize=0.37
    OptionalParentItem=(Type=Bone,Name="spine_04")
    RotationOffset=(X=0,Y=0,Z=0)
   )
   > RigVMVariableNode(Value=@n4.OutputParam) {variable=UpperArm L Up}
   > DeltaRotationZ_1 = RigVMFunctionReferenceNode(Child=head, Parent=spine_05, SourceMax=-80, SourceMin=-70)
   > RigVMVariableNode(Value=&RigVMFunction_MathDoubleMul(B=@DeltaRotationZ_1.Delta Z, A=RigVMVariableNode() {variable=Crouching}.Value).Result) {variable=Head_Rot_Up_Crouch_Extrem}
   &RigUnit_ModifyTransforms(
    ItemToModify=((Item=(Type=Bone,Name="base_scarf_sd_r"),Transform=(Rotation=(X=0,Y=0,Z=0,W=1),Translation=(X=0,Y=0,Z=4),Scale3D=(X=1,Y=1,Z=1))))
    WeightMinimum=0
    ExecutePin=@RigVMFunction_Sequence.B
    Weight=RigVMVariableNode() {variable=Head_Rot_Right}.Value
   )
   n5 = RigVMVariableNode() {variable=Head_Rot_Left}
   > &RigUnit_ModifyTransforms(
    ItemToModify=((Item=(Type=Bone,Name="base_scarf_sd_l"),Transform=(Rotation=(X=0,Y=0,Z=0,W=1),Translation=(X=6,Y=0,Z=0),Scale3D=(X=1,Y=1,Z=1))))
    WeightMinimum=0
    Weight=@n5.Value
   )
   > n6 = &RigUnit_ModifyTransforms(
    ItemToModify="((Item=(Type=Bone,Name=\"base_scarf_sd_l\"),Transform=(Rotation=(X=0,Y=0,Z=0,W=1),Translation=(X=0,Y=-5,Z=-7),Scale3D=(X=1,Y=1,Z=1))),(Item=(Type=Bone,Name=\"base_scarf_bk_c\"),Transform=(Rotation=(X=0,Y=0,Z=0,W=1),Translation=(X=0,Y=0,Z=0),Scale3D=(X=1,Y=1,Z= ...(truncated)"
    WeightMinimum=0
    Weight=&RigVMFunction_MathDoubleRemap(SourceMaximum=1, SourceMinimum=0.75, TargetMaximum=1, TargetMinimum=0, Value=@n5.Value).Result
   )
   RigVMFunction_ControlFlowBranch = &RigVMFunction_ControlFlowBranch(ExecuteContext=@n6.ExecutePin, Condition=RigVMVariableNode() {variable=SkydivingBool}.Value)
   &RigUnit_ModifyTransforms(
    ItemToModify=((Item=(Type=Bone,Name="base_scarf_bk_c"),Transform=(Rotation=(X=0,Y=0.2203,Z=0,W=0.9754),Translation=(X=5,Y=-3.7047,Z=-5),Scale3D=(X=1,Y=1,Z=1))))
    WeightMinimum=0
    ExecutePin=@RigVMFunction_ControlFlowBranch.False
    Weight=RigVMVariableNode() {variable=Head_Rot_Up}.Value
   )
   Multiply_1 = RigVMAggregateNode(
    C=&RigVMFunction_MathDoubleRemap(SourceMaximum=1, SourceMinimum=0, TargetMaximum=0, TargetMinimum=1, Value=RigVMVariableNode() {variable=Emotes}.Value).Result
    A=RigVMVariableNode() {variable=UpperArm L Up}.Value
    B=RigVMVariableNode() {variable=Selected_Pistol}.Value
   )
   > &RigUnit_ModifyTransforms(
    ItemToModify="((Item=(Type=Bone,Name=\"base_scarf_bk_c\"),Transform=(Rotation=(X=0.1736,Y=0,Z=0,W=0.9848),Translation=(X=0,Y=0,Z=-3),Scale3D=(X=1,Y=1,Z=1))),(Item=(Type=Bone,Name=\"base_scarf_sd_r\"),Transform=(Rotation=(X=0,Y=0,Z=0,W=1),Translation=(X=0,Y=0,Z=3),Scale3D=(X ...(truncated)"
    WeightMinimum=0
    Weight=@Multiply_1.Result
   )
   Multiply = RigVMAggregateNode(
    A=RigVMVariableNode() {variable=UpperArm L Up}.Value
    B=RigVMVariableNode() {variable=Crouching}.Value
    C=RigVMVariableNode() {variable=Selected_Pistol}.Value
   )
   > &RigUnit_ModifyTransforms(
    ItemToModify="((Item=(Type=Bone,Name=\"base_scarf_bk_c\"),Transform=(Rotation=(X=0.1736,Y=0,Z=0,W=0.9848),Translation=(X=0,Y=0,Z=1),Scale3D=(X=1,Y=1,Z=1))),(Item=(Type=Bone,Name=\"base_scarf_sd_r\"),Transform=(Rotation=(X=0,Y=0,Z=0,W=1),Translation=(X=0,Y=0,Z=0),Scale3D=(X= ...(truncated)"
    WeightMinimum=0
    Weight=@Multiply.Result
   )
   Multiply_2 = RigVMAggregateNode(
    B=&RigVMFunction_MathDoubleRemap(SourceMaximum=1, SourceMinimum=0, TargetMaximum=0, TargetMinimum=1, Value=RigVMVariableNode() {variable=Selected_Pistol}.Value).Result
    C=&RigVMFunction_MathDoubleRemap(SourceMaximum=1, SourceMinimum=0, TargetMaximum=0, TargetMinimum=1, Value=RigVMVariableNode() {variable=Emotes}.Value).Result
    A=RigVMVariableNode() {variable=UpperArm L Up}.Value
   )
   > &RigUnit_ModifyTransforms(
    ItemToModify="((Item=(Type=Bone,Name=\"base_scarf_bk_c\"),Transform=(Rotation=(X=0.2164,Y=0,Z=0,W=0.9763),Translation=(X=0,Y=-1,Z=2),Scale3D=(X=1,Y=1,Z=1))),(Item=(Type=Bone,Name=\"base_scarf_sd_r\"),Transform=(Rotation=(X=0,Y=0,Z=0,W=1),Translation=(X=0,Y=0,Z=2),Scale3D=(X ...(truncated)"
    WeightMinimum=0
    Weight=@Multiply_2.Result
   )
   > &RigUnit_ModifyTransforms(
    ItemToModify="((Item=(Type=Bone,Name=\"base_scarf_bk_c\"),Transform=(Rotation=(X=-0.2588,Y=0,Z=0,W=0.9659),Translation=(X=0,Y=0,Z=0),Scale3D=(X=1,Y=1,Z=1))),(Item=(Type=Bone,Name=\"base_scarf_sd_r\"),Transform=(Rotation=(X=0,Y=0,Z=0,W=1),Translation=(X=0,Y=5,Z=0),Scale3D=(X ...(truncated)"
    WeightMinimum=0
    Weight=RigVMVariableNode() {variable=Head_Rot_Up_Crouch_Extrem}.Value
   )
   &RigUnit_ModifyTransforms(
    ItemToModify=((Item=(Type=Bone,Name="base_scarf_bk_c"),Transform=(Rotation=(X=0,Y=0,Z=0,W=1),Translation=(X=0,Y=0,Z=-5),Scale3D=(X=1,Y=2,Z=1))))
    WeightMinimum=0
    ExecutePin=@RigVMFunction_ControlFlowBranch.True
    Weight=RigVMVariableNode() {variable=Head_Rot_Up}.Value
   )
   RigVMFunctionReturnNode(ExecuteContext=@RigVMFunction_Sequence.D)
   &RigUnit_ModifyTransforms(
    ItemToModify="((Item=(Type=Bone,Name=\"base_scarf_bk_low_c\"),Transform=(Rotation=(X=0,Y=0.342,Z=0,W=0.9397),Translation=(X=0,Y=0,Z=0),Scale3D=(X=1,Y=1,Z=1))),(Item=(Type=Bone,Name=\"base_scarf_bk_c\"),Transform=(Rotation=(X=-0.2146,Y=0.0283,Z=-0.1274,W=0.9679),Translation= ...(truncated)"
    WeightMinimum=0
    Weight=RigVMVariableNode() {variable=Head_Rot_Up_Crouch}.Value
    ExecutePin=@RigVMFunction_Sequence.C
   )
   n7 = RigVMVariableNode() {variable=Selected_Pistol}
   > &RigUnit_ModifyTransforms(
    ItemToModify="((Item=(Type=Bone,Name=\"base_scarf_bk_low_c\"),Transform=(Rotation=(X=0.0858,Y=0.0151,Z=0.173,W=0.9811),Translation=(X=0,Y=-2,Z=-2),Scale3D=(X=1,Y=1,Z=1))),(Item=(Type=Bone,Name=\"base_scarf_sd_l\"),Transform=(Rotation=(X=0,Y=0,Z=0.2588,W=0.9659),Translation= ...(truncated)"
    WeightMinimum=0
    Weight=&RigVMFunction_MathDoubleMul(
     B=&RigVMFunction_MathDoubleRemap(SourceMaximum=1, SourceMinimum=0, TargetMaximum=0, TargetMinimum=1, Value=@n7.Value).Result
     A=RigVMVariableNode() {variable=Head_Rot_Up_Crouch}.Value
    ).Result
   )
  ]}
 New Function_ContainedGraph {nodes [
   n1 = RigVMFunctionEntryNode()
   RigVMFunction_Sequence = &RigVMFunction_Sequence(ExecuteContext=@n1.ExecuteContext)
   &RigUnit_ParentConstraint(
    AdvancedSettings=(RotationOrderForFilter=XZY)
    Child=(Type=Null,Name="def_bandolier_front_05_null")
    Parents="((Item=(Type=Bone,Name=\"clavicle_l\"),Weight=0.0009),(Item=(Type=Bone,Name=\"deform_pec_r\"),Weight=0.0013),(Item=(Type=Bone,Name=\"spine_05\"),Weight=0.005),(Item=(Type=Bone,Name=\"pelvis\"),Weight=0.0345),(Item=(Type=Bone,Name=\"spine_01\"),Weight=0.0848),(Item=( ...(truncated)"
    Weight=1
    ExecutePin=@RigVMFunction_Sequence.A
   )
   > &RigUnit_ParentConstraint(
    AdvancedSettings=(RotationOrderForFilter=XZY)
    Child=(Type=Null,Name="def_bandolier_front_06_null")
    Parents="((Item=(Type=Bone,Name=\"clavicle_r\"),Weight=0.0002),(Item=(Type=Bone,Name=\"dyn_fr_shirt_lower_bagginess_r\"),Weight=0.0003),(Item=(Type=Bone,Name=\"spine_04\"),Weight=0.0722),(Item=(Type=Bone,Name=\"pelvis\"),Weight=0.1341),(Item=(Type=Bone,Name=\"spine_01\"),Wei ...(truncated)"
    Weight=1
   )
   > &RigUnit_ParentConstraint(
    AdvancedSettings=(RotationOrderForFilter=XZY)
    Child=(Type=Null,Name="def_bandolier_front_07_null")
    Parents="((Item=(Type=Bone,Name=\"spine_04\"),Weight=0.0214),(Item=(Type=Bone,Name=\"dyn_fr_shirt_lower_bagginess_l\"),Weight=0.0778),(Item=(Type=Bone,Name=\"spine_01\"),Weight=0.1005),(Item=(Type=Bone,Name=\"spine_03\"),Weight=0.1412),(Item=(Type=Bone,Name=\"spine_02\"),Wei ...(truncated)"
    Weight=1
   )
   > &RigUnit_ParentConstraint(
    AdvancedSettings=(RotationOrderForFilter=XZY)
    Child=(Type=Null,Name="def_bandolier_front_08_null")
    Parents="((Item=(Type=Bone,Name=\"spine_04\"),Weight=0.0021),(Item=(Type=Bone,Name=\"dyn_sd_shirt_lower_bagginess_l\"),Weight=0.0127),(Item=(Type=Bone,Name=\"spine_01\"),Weight=0.0146),(Item=(Type=Bone,Name=\"spine_03\"),Weight=0.0169),(Item=(Type=Bone,Name=\"spine_02\"),Wei ...(truncated)"
    Weight=1
   )
   > &RigUnit_ParentConstraint(
    AdvancedSettings=(RotationOrderForFilter=XZY)
    Child=(Type=Null,Name="def_bandolier_rear_05_null")
    Parents="((Item=(Type=Bone,Name=\"pelvis\"),Weight=0.0245),(Item=(Type=Bone,Name=\"clavicle_l\"),Weight=0.0414),(Item=(Type=Bone,Name=\"clavicle_r\"),Weight=0.0472),(Item=(Type=Bone,Name=\"spine_01\"),Weight=0.0598),(Item=(Type=Bone,Name=\"spine_05\"),Weight=0.0664),(Item=(T ...(truncated)"
    Weight=1
   )
   > &RigUnit_ParentConstraint(
    AdvancedSettings=(RotationOrderForFilter=XZY)
    Child=(Type=Null,Name="def_bandolier_rear_06_null")
    Parents="((Item=(Type=Bone,Name=\"spine_05\"),Weight=0.0021),(Item=(Type=Bone,Name=\"clavicle_l\"),Weight=0.0144),(Item=(Type=Bone,Name=\"spine_03\"),Weight=0.0297),(Item=(Type=Bone,Name=\"pelvis\"),Weight=0.0837),(Item=(Type=Bone,Name=\"spine_04\"),Weight=0.1032),(Item=(Typ ...(truncated)"
    Weight=1
   )
   > &RigUnit_ParentConstraint(
    AdvancedSettings=(RotationOrderForFilter=XZY)
    Child=(Type=Null,Name="def_bandolier_rear_07_null")
    Parents="((Item=(Type=Bone,Name=\"back_fix\"),Weight=0.0039),(Item=(Type=Bone,Name=\"dyn_sd_shirt_lower_bagginess_l\"),Weight=0.0235),(Item=(Type=Bone,Name=\"spine_04\"),Weight=0.0277),(Item=(Type=Bone,Name=\"spine_01\"),Weight=0.1047),(Item=(Type=Bone,Name=\"spine_03\"),Wei ...(truncated)"
    Weight=1
   )
   > &RigUnit_ParentConstraint(
    AdvancedSettings=(RotationOrderForFilter=XZY)
    Child=(Type=Null,Name="def_bandolier_rear_08_null")
    Parents=((Item=(Type=Bone,Name="dyn_bk_belt_r"),Weight=0.0051),(Item=(Type=Bone,Name="dyn_sd_belt_l"),Weight=0.3106),(Item=(Type=Bone,Name="dyn_bk_belt_l"),Weight=0.6843))
    Weight=1
   )
   RigVMFunctionReturnNode(ExecuteContext=@RigVMFunction_Sequence.B)
   &RigUnit_ParentConstraint(
    AdvancedSettings=(RotationOrderForFilter=XZY)
    Child=(Type=Null,Name="def_bandolier_front_08_null")
    ExecutePin=()
    Parents="((Item=(Type=Bone,Name=\"spine_03\"),Weight=0.0002),(Item=(Type=Bone,Name=\"spine_02\"),Weight=0.0002),(Item=(Type=Bone,Name=\"dyn_sd_shirt_lower_bagginess_l\"),Weight=0.0004),(Item=(Type=Bone,Name=\"dyn_fr_shirt_lower_bagginess_l\"),Weight=0.0067),(Item=(Type=Bon ...(truncated)"
    Weight=1
   )
   > &RigUnit_ParentConstraint(
    AdvancedSettings=(RotationOrderForFilter=XZY)
    Child=(Type=Null,Name="def_bandolier_front_10")
    Parents=((Item=(Type=Bone,Name="dyn_fr_belt_l"),Weight=0.3578),(Item=(Type=Bone,Name="dyn_sd_belt_l"),Weight=0.6422))
    Weight=1
   )
   &RigUnit_ParentConstraint(
    AdvancedSettings=(RotationOrderForFilter=XZY)
    Child=(Type=Bone,Name="def_bandolier_rear_09")
    ExecutePin=()
    Parents=((Item=(Type=Bone,Name="dyn_fr_belt_l"),Weight=0.0001),(Item=(Type=Bone,Name="dyn_bk_belt_r"),Weight=0.0002),(Item=(Type=Bone,Name="dyn_bk_belt_l"),Weight=0.4281),(Item=(Type=Bone,Name="dyn_sd_belt_l"),Weight=0.5716))
    Weight=1
   )
   &RigUnit_ParentConstraint(
    AdvancedSettings=(RotationOrderForFilter=XZY)
    Child=(Type=Bone,Name="def_bandolier_rear_10")
    ExecutePin=()
    Parents=((Item=(Type=Bone,Name="dyn_bk_belt_r"),Weight=0.0001),(Item=(Type=Bone,Name="dyn_fr_belt_l"),Weight=0.0021),(Item=(Type=Bone,Name="dyn_bk_belt_l"),Weight=0.1779),(Item=(Type=Bone,Name="dyn_sd_belt_l"),Weight=0.8199))
    Weight=1
   )
  ]}
 New Function_ContainedGraph {nodes [
   n1 = RigVMFunctionEntryNode()
   RigVMFunction_Sequence = &RigVMFunction_Sequence(ExecuteContext=@n1.ExecuteContext)
   &RigUnit_ModifyTransforms(
    ItemToModify=((Item=(Type=Bone,Name="back_fix"),Transform=(Rotation=(X=0,Y=0,Z=0,W=-1),Translation=(X=-3,Y=0,Z=0),Scale3D=(X=1,Y=1,Z=1))))
    Weight=1
    WeightMinimum=0
    ExecutePin=@RigVMFunction_Sequence.A
   )
   > &RigUnit_ModifyTransforms(
    ItemToModify="((Item=(Type=Null,Name=\"def_bandolier_front_03_null\"),Transform=(Rotation=(X=0,Y=0,Z=0,W=1),Translation=(X=0,Y=0,Z=-1),Scale3D=(X=1,Y=1,Z=1))),(Item=(Type=Null,Name=\"def_bandolier_front_04_null\"),Transform=(Rotation=(X=0,Y=0,Z=0,W=1),Translation=(X=0,Y=0,Z ...(truncated)"
    Weight=1
    WeightMinimum=0
   )
   > &RigUnit_ParentConstraint(
    AdvancedSettings=(InterpolationType=Average,RotationOrderForFilter=XZY)
    Child=(Type=Null,Name="base_bag_null")
    Parents=((Item=(Type=Bone,Name="pelvis"),Weight=1),(Item=(Type=Bone,Name="thigh_l"),Weight=1))
    Weight=1
   )
   > &RigUnit_ModifyTransforms(
    ItemToModify="((Item=(Type=Bone,Name=\"def_bandolier_rear_10\"),Transform=(Rotation=(X=0,Y=0,Z=-0.7071,W=-0.7071),Translation=(X=0,Y=0,Z=0),Scale3D=(X=1,Y=1,Z=1))),(Item=(Type=Bone,Name=\"def_bandolier_front_10\"),Transform=(Rotation=(X=0,Y=0,Z=0.7071,W=0.7071),Translation= ...(truncated)"
    WeightMinimum=0
    Weight=RigVMVariableNode() {variable=Side_Leg_L}.Value
   )
   Multiply = &RigVMFunction_MathDoubleMul(A=RigVMVariableNode() {variable=Front_Leg_L}.Value, B=RigVMVariableNode() {variable=Side_Leg_L}.Value)
   > &RigUnit_ModifyTransforms(
    ItemToModify="((Item=(Type=Bone,Name=\"def_bandolier_rear_10\"),Transform=(Rotation=(X=0,Y=0,Z=-0.4226,W=0.9063),Translation=(X=0,Y=0,Z=0),Scale3D=(X=1,Y=1,Z=1))),(Item=(Type=Bone,Name=\"def_bandolier_front_10\"),Transform=(Rotation=(X=0,Y=0,Z=-0.4226,W=0.9063),Translation= ...(truncated)"
    WeightMinimum=0
    Weight=@Multiply.Result
   )
   > &RigUnit_ModifyTransforms(
    ItemToModify="((Item=(Type=Null,Name=\"def_bandolier_front_06_null\"),Transform=(Rotation=(X=0,Y=0,Z=0,W=1),Translation=(X=0,Y=3,Z=0),Scale3D=(X=1,Y=1,Z=1))),(Item=(Type=Null,Name=\"def_bandolier_front_07_null\"),Transform=(Rotation=(X=0,Y=0,Z=0,W=1),Translation=(X=0,Y=0,Z= ...(truncated)"
    WeightMinimum=0
    Weight=RigVMVariableNode() {variable=Crouching}.Value
   )
   > SplineIK = RigVMFunctionReferenceNode(
    Primary Axis=(X=1,Y=0,Z=0)
    Samples Per Segment=3
    Secondary Spline Direction=(X=0,Y=-1,Z=0)
    Spline Mode=BSpline
    Stretch=true
    Up Axis=(X=0,Y=-1,Z=0)
    debug=false
    Bones=RigVMVariableNode() {variable=Bandolier_Front_JointsNoRBAN}.Value
    Controls=RigVMVariableNode() {variable=Bandolier_Front_CtrlsNoRBAN}.Value
   )
   > SplineIK_2 = RigVMFunctionReferenceNode(
    Primary Axis=(X=1,Y=0,Z=0)
    Samples Per Segment=3
    Secondary Spline Direction=(X=0,Y=1,Z=0)
    Spline Mode=BSpline
    Stretch=true
    Up Axis=(X=0,Y=1,Z=0)
    debug=false
    Controls=RigVMVariableNode() {variable=Bandolier_Rear_CtrlsNoRBAN}.Value
    Bones=RigVMVariableNode() {variable=Bandolier_Rear_JointsNoRBAN}.Value
   )
   RigVMFunctionReturnNode(ExecuteContext=@RigVMFunction_Sequence.B)
  ]}
 New Function_ContainedGraph {nodes [
   n1 = RigVMFunctionEntryNode()
   RigVMFunction_MathQuaternionToEuler = &RigVMFunction_MathQuaternionToEuler(
    RotationOrder=ZYX
    Value=&RigUnit_GetRelativeTransformForItem(
     Child=(Type=Bone,Name="None")
     Parent=(Type=Bone,Name="None")
     bChildInitial=False
     bParentInitial=False
     Child.Name=@n1.Child
     Parent.Name=@n1.Parent
    ).RelativeTransform.Rotation
   )
   Remap = &RigVMFunction_MathDoubleRemap(
    TargetMaximum=1
    TargetMinimum=0
    Value=@RigVMFunction_MathQuaternionToEuler.Result.Z
    SourceMaximum=@n1.SourceMax
    SourceMinimum=@n1.SourceMin
   )
   > RigVMFunctionReturnNode(Delta Z=@Remap.Result)
  ]}
 New Function_ContainedGraph {nodes [
   n1 = RigVMFunctionEntryNode()
   Clamp = &RigVMFunction_MathIntClamp(Value=@n1.primAxis)
   RigVMFunction_MathBoolToInteger = &RigVMFunction_MathBoolToInteger(Value=@n1.onlyTwist)
   RigVMFunction_MathQuaternionSwingTwist = &RigVMFunction_MathQuaternionSwingTwist(
    TwistAxis=@n1.twistAxis
    Input=&RigUnit_GetRelativeTransformForItem(
     Child=(Type=Bone,Name="None")
     Parent=(Type=Bone,Name="None")
     bChildInitial=False
     bParentInitial=False
     Parent.Type=@n1.ParentType
     Child.Name=@n1.child
     Parent.Name=@n1.parent
    ).RelativeTransform.Rotation
   )
   RigVMFunction_MathQuaternionToEuler = &RigVMFunction_MathQuaternionToEuler(
    Value=RigVMDispatchNode(
     Values=((),())
     Index=@RigVMFunction_MathBoolToInteger.Result
     Values.0=@RigVMFunction_MathQuaternionSwingTwist.Swing
     Values.1=@RigVMFunction_MathQuaternionSwingTwist.Twist
    ) {template="DISPATCH_RigVMDispatch_SelectInt32(in Index,in Values,out Result)"}.Result
    RotationOrder=@n1.rotationOrder
   )
   DISPATCH_RigVMDispatch_SelectInt32 = RigVMDispatchNode(
    Values=(0,0,0)
    Index=@Clamp.Result
    Values.0=@RigVMFunction_MathQuaternionToEuler.Result.X
    Values.1=@RigVMFunction_MathQuaternionToEuler.Result.Y
    Values.2=@RigVMFunction_MathQuaternionToEuler.Result.Z
   ) {template="DISPATCH_RigVMDispatch_SelectInt32(in Index,in Values,out Result)"}
   Remap = &RigVMFunction_MathDoubleRemap(
    TargetMaximum=1
    TargetMinimum=0
    bClamp=false
    Value=@DISPATCH_RigVMDispatch_SelectInt32.Result
    SourceMaximum=@n1.max
    SourceMinimum=@n1.min
   )
   > RigVMFunctionReturnNode(ouputWeight=&RigVMFunction_MathDoubleClamp(Value=@Remap.Result).Result)
  ]}
 RigVMFunctionLibrary {nodes [
   BandolierPinNoRBAN = RigVMCollapseNode(ExecuteContext=())
   BandolierPinSetup = RigVMCollapseNode(ExecuteContext=())
   BeltSetup = RigVMCollapseNode(ExecuteContext=())
   ControlsSetup = RigVMCollapseNode(ControlsDebug=False, ExecuteContext=())
   DeltaRotation = RigVMCollapseNode(ExecuteContext=(), twistAxis=(X=0,Y=0,Z=0))
   DeltaRotationZ = RigVMCollapseNode(ExecuteContext=())
   HeadSetup = RigVMCollapseNode(ExecuteContext=())
   SolveBandolier = RigVMCollapseNode(ExecuteContext=(), Name=base_bag_null)
   SolveBandolierNoRBAN = RigVMCollapseNode(ExecuteContext=())
  ]}
 RigVMFunction_Sequence_1_ContainedGraph {nodes [
   n1 = RigVMFunctionEntryNode()
   RigVMFunction_Sequence = &RigVMFunction_Sequence(ExecuteContext=@n1.ExecuteContext)
   n2 = &RigVMFunction_Sequence(ExecuteContext=@RigVMFunction_Sequence.B)
   n3 = &RigVMFunction_Sequence(ExecuteContext=@n2.B)
   n4 = &RigVMFunction_Sequence(ExecuteContext=@n3.B)
   n5 = &RigVMFunction_Sequence(ExecuteContext=@n4.B)
   RigVMFunctionReturnNode(A=@RigVMFunction_Sequence.A, B=@n2.A, C=@n3.A, D=@n4.A, E=@n5.A, F=@n5.B)
  ]}
 RigVMFunction_Sequence_1_ContainedGraph {nodes [
   n1 = RigVMFunctionEntryNode()
   RigVMFunction_Sequence = &RigVMFunction_Sequence(ExecuteContext=@n1.ExecuteContext)
   n2 = &RigVMFunction_Sequence(ExecuteContext=@RigVMFunction_Sequence.B)
   n3 = &RigVMFunction_Sequence(ExecuteContext=@n2.B)
   RigVMFunctionReturnNode(A=@RigVMFunction_Sequence.A, B=@n2.A, C=@n3.A, D=@n3.B)
  ]}
 RigVMModel {nodes [
   RigUnit_BeginExecution = &RigUnit_BeginExecution()
   ControlsSetup = RigVMFunctionReferenceNode(ControlsDebug=false, ExecuteContext=@RigUnit_BeginExecution.ExecutePin)
   > FN_DeformRig_FNC = RigVMFunctionReferenceNode(
    DeformRig_ Curve="(deformRig_l_upperarm_up_alpha_72_#1=False,deformRig_r_upperarm_up_alpha_73_#2=true,deformRig_l_upperarm_fr_alpha_74_#3=true,deformRig_r_upperarm_fr_alpha_75_CB51B80C ...(truncated)"
    DeformRig_Bool="(upperarm_1_#4=True,elbow_3_#5=True,wrist_7_#6=True,groin_8_#7=True,glute_17_#8=True,knee_9_40C408324CA261 ...(truncated)"
    DeformRig_Debug="(upperarm_debug=False,wrist_debug=False,lat_debug=False,groinGlute_debug=False,upperarm_debug_18_#4=False,wrist_debug_20_#6=False,groinGlute_debug_27_#7=False,lat_deb ...(truncated)"
    DeformRig_Value="(upperarm_up_18_#9=(Translation=(X=0,Y=0,Z=8),Scale3D=(X=0.7,Y=1,Z=1)),upperarm_fwd_fr_19_#10=(Translation=(X=0,Y=10,Z=0),Scale3D=(X=0.5,Y=1,Z=1)),upperarm_fwd_bk_20_9973AAD84EE0A1E349BC9E85067D4A3 ...(truncated)"
   )
   > BeltSetup = RigVMFunctionReferenceNode()
   > HeadSetup = RigVMFunctionReferenceNode()
   RigVMFunction_ControlFlowBranch = &RigVMFunction_ControlFlowBranch(ExecuteContext=@HeadSetup.ExecuteContext, Condition=RigVMVariableNode() {variable=RBAN Enable}.Value)
   BandolierPinSetup_1 = RigVMFunctionReferenceNode(ExecuteContext=@RigVMFunction_ControlFlowBranch.False)
   > BandolierPinNoRBAN = RigVMFunctionReferenceNode()
   > SolveBandolierNoRBAN = RigVMFunctionReferenceNode()
   BandolierPinSetup = RigVMFunctionReferenceNode(ExecuteContext=@RigVMFunction_ControlFlowBranch.True)
   > SolveBandolier = RigVMFunctionReferenceNode(Name=base_bag_null)
  ]}
]
hierarchy [
 root {loc=<0 0 0>, rot=<0 0 0 1>}
 pelvis {parent=root, loc=<0 -0.7071 93.1191>, rot=<0.0127 0.707 -0.0127 -0.707>}
 spine_01 {parent=pelvis, loc=<6.7494 -0.6388 0>, rot=<0 0 0.034 -0.9994>}
 spine_02 {parent=spine_01, loc=<7.2518 -0.035 0>, rot=<0 0 0.0238 -0.9997>}
 spine_03 {parent=spine_02, loc=<6.0104 0 0>, rot=<0 0 0.0149 -0.9999>}
 spine_04 {parent=spine_03, loc=<8.1024 0 0>, rot=<0 0 -0.0132 -0.9999>}
 spine_05 {parent=spine_04, loc=<10.0146 0 0>, rot=<0 0 -0.0015 -1>}
 neck_01 {parent=spine_05, loc=<11.3138 0.1529 0>, rot=<0 0 -0.2003 -0.9797>}
 neck_02 {parent=neck_01, loc=<4.5627 0 0>, rot=<0 0 0.0504 -0.9987>}
 head {parent=neck_02, loc=<5.3914 0 0>, rot=<0 0 0.115 -0.9934>}
 base_hair {parent=head, loc=<11.452 9.8368 -1.5418>, rot=<-0.1202 0.7051 -0.3657 -0.5955>}
 dyn_hair_fr_01 {parent=base_hair, loc=<0 0 0>, rot=<0 0 0 1>}
 dyn_hair_fr_02 {parent=dyn_hair_fr_01, loc=<4.045 0 0>, rot=<0 0 0 1>}
 dyn_hair_sd_01_l {parent=head, loc=<6.0054 5.8453 -6.7408>, rot=<0.2967 -0.6808 0.3494 -0.5713>}
 dyn_hair_sd_02_l {parent=head, loc=<7.8363 2.1546 -7.9208>, rot=<0.2924 -0.6827 0.3458 -0.5735>}
 dyn_hair_sd_03_l {parent=head, loc=<9.5108 1.2924 -8.0821>, rot=<-0.2949 -0.712 -0.2066 -0.6028>}
 dyn_hair_sd_04_l {parent=head, loc=<9.0927 -1.4778 -8.5602>, rot=<0.177 -0.6727 0.1481 -0.7029>}
 dyn_hair_sd_05_l {parent=head, loc=<6.5495 -1.6021 -8.4068>, rot=<0.2445 -0.7304 0.2394 -0.5912>}
 dyn_hair_bk_01 {parent=head, loc=<1.0738 -1.4627 -6.6404>, rot=<0.1279 -0.8831 0.2401 -0.3823>}
 dyn_hair_bk_02 {parent=head, loc=<0.5833 -4.3843 -5.5563>, rot=<0.3596 -0.9103 0.1122 0.1718>}
 dyn_hair_bk_03 {parent=head, loc=<3.83 -6.9362 -5.3163>, rot=<0.0599 -0.7897 -0.3903 -0.4695>}
 dyn_hair_bk_04 {parent=head, loc=<-0.2851 -6.2463 -2.429>, rot=<0.3901 -0.8719 -0.2959 0.0022>}
 dyn_hair_bk_05 {parent=head, loc=<2.4778 -7.6159 -0.836>, rot=<0.2144 -0.9153 -0.3134 -0.1346>}
 dyn_hair_bk_06 {parent=head, loc=<1.7937 -7.3676 0.7425>, rot=<0.4929 -0.8041 0.2524 0.2162>}
 dyn_hair_bk_07 {parent=head, loc=<3.7322 -7.6734 2.9731>, rot=<0.2051 -0.8408 -0.3912 -0.3131>}
 dyn_hair_bk_08 {parent=head, loc=<0.251 -5.3967 3.5267>, rot=<0.1631 -0.9583 0.1561 -0.1754>}
 dyn_hair_bk_09 {parent=head, loc=<7.6326 -7.8468 2.9379>, rot=<0.2852 -0.5684 0.5988 -0.4869>}
 dyn_hair_bk_10 {parent=head, loc=<11.7821 -7.2415 -1.9942>, rot=<0.4388 -0.5299 0.6193 -0.3783>}
 dyn_hair_bk_11 {parent=head, loc=<8.4546 -6.5869 -5.9163>, rot=<0.5564 -0.4283 0.4052 -0.5854>}
 dyn_hair_bk_12 {parent=head, loc=<12.8558 -4.1353 -5.6767>, rot=<0.4183 -0.57 0.2386 -0.6657>}
 dyn_hair_sd_01_r {parent=head, loc=<6.923 -3.929 7.3063>, rot=<0.6259 -0.2049 0.6939 -0.291>}
 dyn_hair_sd_02_r {parent=head, loc=<3.5754 -4.426 6.8559>, rot=<0.5043 0.1108 0.8556 0.0372>}
 dyn_hair_sd_03_r {parent=head, loc=<-0.047 -2.3596 5.8649>, rot=<0.3963 -0.2086 0.8088 -0.3813>}
 dyn_hair_sd_04_r {parent=head, loc=<1.0314 -1.2397 6.6393>, rot=<0.4578 -0.1438 0.8308 -0.282>}
 dyn_hair_sd_05_r {parent=head, loc=<7.5719 2.4633 8.0629>, rot=<0.6308 -0.1695 0.6349 -0.4126>}
 dyn_hair_sd_06_r {parent=head, loc=<6.7131 7.2709 5.6348>, rot=<-0.1271 -0.7824 -0.5478 -0.2676>}
 dyn_hair_sd_07_r {parent=head, loc=<9.3394 9.0258 4.375>, rot=<0.3647 -0.1661 0.9154 0.0373>}
 dyn_hair_sd_08_r {parent=head, loc=<8.957 -7.0099 5.0133>, rot=<0.409 -0.2815 0.8412 -0.2143>}
 dyn_hair_sd_09_r {parent=head, loc=<13.7088 -3.2953 4.7602>, rot=<0.5438 -0.4473 0.4955 -0.5086>}
 dyn_hair_sd_10_r {parent=head, loc=<9.534 0.3535 7.8937>, rot=<0.579 0.4006 0.6248 0.3374>}
 attach_cape {parent=spine_05, loc=<9.0163 -11.7139 0.0001>, rot=<0.0616 -0.7044 -0.0616 -0.7044>}
 cape {parent=attach_cape, type=Null, loc=<0 0 0>, rot=<0 0 0 1>}
 attach_backpack {parent=spine_05, loc=<-0.1239 -13.4729 0.0001>, rot=<-0.0093 -0.707 0.0093 -0.707>}
 Backpack_BR {parent=attach_backpack, type=Null, loc=<0 -0.5 0>, rot=<0 0 0 1>}
 clavicle_l {parent=spine_05, loc=<4.3575 -0.1466 -2.3223>, rot=<0.0465 -0.7289 -0.009 -0.6829>}
 upperarm_l {parent=clavicle_l, loc=<15.5111 0 0>, rot=<0.0226 -0.3234 -0.0088 -0.9459>}
 lowerarm_l {parent=upperarm_l, loc=<26.6514 0 0>, rot=<0.0391 -0.0099 -0.2453 -0.9686>}
 hand_l {parent=lowerarm_l, loc=<22.3491 0 0>, rot=<-0.7081 -0.0507 0.0975 -0.6975>}
 thumb_01_l {parent=hand_l, loc=<1.3118 -0.3191 -2.459>, rot=<0.6546 -0.4042 0.0664 -0.6354>}
 thumb_02_l {parent=thumb_01_l, loc=<5.9336 -0.0448 0.0401>, rot=<-0.0037 0.0213 0.298 -0.9543>}
 thumb_03_l {parent=thumb_02_l, loc=<2.2223 0.3682 -0.0907>, rot=<0.0118 -0.013 0.0066 -0.9998>}
 index_metacarpal_l {parent=hand_l, loc=<1.6535 0.0183 -2.96>, rot=<0.0158 -0.0124 -0.0173 -0.9996>}
 index_01_l {parent=index_metacarpal_l, loc=<8.3008 0.0357 -0.0524>, rot=<-0.0044 0.0485 0.0924 -0.9945>}
 index_02_l {parent=index_01_l, loc=<3.3587 -0.3004 0.0452>, rot=<-0.002 0.0027 0.3225 -0.9466>}
 index_03_l {parent=index_02_l, loc=<2.5578 -0.123 -0.0252>, rot=<-0.002 0.0067 0.2558 -0.9667>}
 middle_metacarpal_l {parent=hand_l, loc=<2.1856 -0.1725 -0.5984>, rot=<0.006 0.0168 -0.0329 -0.9993>}
 middle_01_l {parent=middle_metacarpal_l, loc=<7.3275 -0.2447 0.0129>, rot=<-0.0208 0.0042 0.0858 -0.9961>}
 middle_02_l {parent=middle_01_l, loc=<4.0123 0.1401 0.1175>, rot=<-0.0218 0.0258 0.3512 -0.9357>}
 middle_03_l {parent=middle_02_l, loc=<2.6077 0.0534 0.1373>, rot=<0.0082 0.0188 0.1708 -0.9851>}
 ring_metacarpal_l {parent=hand_l, loc=<2.2074 -0.3625 1.3638>, rot=<-0.0284 0.0722 -0.0458 -0.9959>}
 ring_01_l {parent=ring_metacarpal_l, loc=<7.0295 -0.6197 -0.1446>, rot=<-0.0489 -0.0252 0.1892 -0.9804>}
 ring_02_l {parent=ring_01_l, loc=<3.6174 0.4647 -0.1027>, rot=<0 0 0.2806 -0.9598>}
 ring_03_l {parent=ring_02_l, loc=<2.4152 0.114 0.0199>, rot=<0 0 0.1189 -0.9929>}
 pinky_metacarpal_l {parent=hand_l, loc=<1.7616 -0.9753 3.3814>, rot=<-0.0397 0.0814 -0.069 -0.9935>}
 pinky_01_l {parent=pinky_metacarpal_l, loc=<7.1709 -0.4831 -0.1094>, rot=<-0.0654 -0.0035 0.2242 -0.9723>}
 pinky_02_l {parent=pinky_01_l, loc=<2.8274 0.0893 -0.1043>, rot=<0.0037 0.0011 0.2854 -0.9584>}
 pinky_03_l {parent=pinky_02_l, loc=<2.2046 0.0307 -0.0078>, rot=<-0.0002 0.0098 0.1445 -0.9895>}
 deform_wrist_dn_base_l {parent=hand_l, loc=<0 -2.2342 0>, rot=<0 0 0 -1>}
 deform_wrist_dn_l {parent=deform_wrist_dn_base_l, loc=<0 0 0>, rot=<0 0 0 1>}
 deform_wrist_up_base_l {parent=hand_l, loc=<0 2.2682 0>, rot=<0 0 0 -1>}
 deform_wrist_up_l {parent=deform_wrist_up_base_l, loc=<0 0 0>, rot=<0 0 0 1>}
 weapon_l {parent=hand_l, loc=<-0.1974 -3.4111 0.3811>, rot=<0 0 0 1>}
 hand_attach_l {parent=hand_l, loc=<8.5 -3 0>, rot=<0.7071 0.7071 0 0>}
 lowerarm_twist_02_l {parent=lowerarm_l, loc=<9.6199 0 0>, rot=<0 0 0 -1>}
 lowerarm_twist_01_l {parent=lowerarm_l, loc=<17.8549 0 0>, rot=<0 0 0 -1>}
 deform_elbow_in_base_l {parent=lowerarm_l, loc=<0.2112 5.259 0.8274>, rot=<0 0 0 -1>}
 deform_elbow_in_l {parent=deform_elbow_in_base_l, loc=<0 0 0>, rot=<0 0 0 1>}
 deform_elbow_out_base_l {parent=lowerarm_l, loc=<0 -4 0>, rot=<0 0 0 -1>}
 deform_elbow_out_l {parent=deform_elbow_out_base_l, loc=<0 0 0>, rot=<0 0 0 1>}
 upperarm_twist_01_l {parent=upperarm_l, loc=<0.0041 0 0>, rot=<0 0 0 -1>}
 deform_upperarm_bk_base_l {parent=upperarm_twist_01_l, loc=<1.0813 -7.2193 -1.9453>, rot=<0 0 0 -1>}
 deform_upperarm_bk_l {parent=deform_upperarm_bk_base_l, loc=<0 0 0>, rot=<0 0 0 1>}
 deform_upperarm_fr_base_l {parent=upperarm_twist_01_l, loc=<-0.3691 7.0506 -2.765>, rot=<0 0 0 -1>}
 deform_upperarm_fr_l {parent=deform_upperarm_fr_base_l, loc=<0 0 0>, rot=<0 0 0 1>}
 deform_upperarm_up_base_l {parent=upperarm_twist_01_l, loc=<-3.2216 -1.1391 3.3311>, rot=<0 0 0 1>}
 deform_upperarm_up_l {parent=deform_upperarm_up_base_l, loc=<0 0 0>, rot=<0 0 0 1>}
 upperarm_twist_02_l {parent=upperarm_l, loc=<13.7941 0 0>, rot=<0 0 0 -1>}
 clavicle_r {parent=spine_05, loc=<4.3575 -0.1466 2.3223>, rot=<0.7289 0.0465 -0.6829 0.009>}
 upperarm_r {parent=clavicle_r, loc=<-15.5111 0 0>, rot=<0.0226 -0.3234 -0.0088 -0.9459>}
 lowerarm_r {parent=upperarm_r, loc=<-26.6514 0 0>, rot=<0.0391 -0.0099 -0.2453 -0.9686>}
 hand_r {parent=lowerarm_r, loc=<-22.3491 0 0>, rot=<-0.7081 -0.0507 0.0975 -0.6975>}
 thumb_01_r {parent=hand_r, loc=<-1.3118 0.3191 2.459>, rot=<0.6546 -0.4042 0.0664 -0.6354>}
 thumb_02_r {parent=thumb_01_r, loc=<-5.9336 0.0448 -0.0401>, rot=<-0.0037 0.0213 0.298 -0.9543>}
 thumb_03_r {parent=thumb_02_r, loc=<-2.2223 -0.3682 0.0907>, rot=<0.0118 -0.013 0.0066 -0.9998>}
 index_metacarpal_r {parent=hand_r, loc=<-1.6535 -0.0183 2.96>, rot=<0.0158 -0.0124 -0.0173 -0.9996>}
 index_01_r {parent=index_metacarpal_r, loc=<-8.3008 -0.0357 0.0524>, rot=<-0.0044 0.0485 0.0924 -0.9945>}
 index_02_r {parent=index_01_r, loc=<-3.3587 0.3004 -0.0452>, rot=<-0.002 0.0027 0.3225 -0.9466>}
 index_03_r {parent=index_02_r, loc=<-2.5578 0.123 0.0252>, rot=<-0.002 0.0067 0.2558 -0.9667>}
 middle_metacarpal_r {parent=hand_r, loc=<-2.1856 0.1725 0.5984>, rot=<0.006 0.0168 -0.0329 -0.9993>}
 middle_01_r {parent=middle_metacarpal_r, loc=<-7.3275 0.2447 -0.0129>, rot=<-0.0208 0.0042 0.0858 -0.9961>}
 middle_02_r {parent=middle_01_r, loc=<-4.0123 -0.1401 -0.1175>, rot=<-0.0218 0.0258 0.3512 -0.9357>}
 middle_03_r {parent=middle_02_r, loc=<-2.6077 -0.0534 -0.1373>, rot=<0.0082 0.0188 0.1708 -0.9851>}
 ring_metacarpal_r {parent=hand_r, loc=<-2.2074 0.3625 -1.3638>, rot=<-0.0284 0.0722 -0.0458 -0.9959>}
 ring_01_r {parent=ring_metacarpal_r, loc=<-7.0295 0.6197 0.1446>, rot=<-0.0489 -0.0252 0.1892 -0.9804>}
 ring_02_r {parent=ring_01_r, loc=<-3.6174 -0.4647 0.1027>, rot=<0 0 0.2806 -0.9598>}
 ring_03_r {parent=ring_02_r, loc=<-2.4152 -0.114 -0.0199>, rot=<0 0 0.1189 -0.9929>}
 pinky_metacarpal_r {parent=hand_r, loc=<-1.7616 0.9753 -3.3814>, rot=<-0.0397 0.0814 -0.069 -0.9935>}
 pinky_01_r {parent=pinky_metacarpal_r, loc=<-7.1709 0.4831 0.1094>, rot=<-0.0654 -0.0035 0.2242 -0.9723>}
 pinky_02_r {parent=pinky_01_r, loc=<-2.8274 -0.0893 0.1043>, rot=<0.0037 0.0011 0.2854 -0.9584>}
 pinky_03_r {parent=pinky_02_r, loc=<-2.2046 -0.0307 0.0078>, rot=<-0.0002 0.0098 0.1445 -0.9895>}
 deform_wrist_dn_base_r {parent=hand_r, loc=<-0.0004 2.2345 -0.0001>, rot=<0 0 0 -1>}
 deform_wrist_dn_r {parent=deform_wrist_dn_base_r, loc=<0 0 0>, rot=<0 0 0 1>}
 deform_wrist_up_base_r {parent=hand_r, loc=<0.0003 -2.2684 0.0001>, rot=<0 0 0 -1>}
 deform_wrist_up_r {parent=deform_wrist_up_base_r, loc=<0 0 0>, rot=<0 0 0 1>}
 weapon_r {parent=hand_r, loc=<0.1974 3.4111 -0.3811>, rot=<0 0 0 1>}
 hand_attach_r {parent=hand_r, loc=<-8.5001 2.9999 0>, rot=<0 0 -0.7071 -0.7071>}
 lowerarm_twist_02_r {parent=lowerarm_r, loc=<-9.6199 0 0>, rot=<0 0 0 -1>}
 lowerarm_twist_01_r {parent=lowerarm_r, loc=<-17.8549 0 0>, rot=<0 0 0 -1>}
 deform_elbow_in_base_r {parent=lowerarm_r, loc=<-0.2111 -5.2591 -0.8275>, rot=<0 0 0 -1>}
 deform_elbow_in_r {parent=deform_elbow_in_base_r, loc=<-0.0001 0 0.0002>, rot=<0 0 0 1>}
 deform_elbow_out_base_r {parent=lowerarm_r, loc=<0 4 -0.0001>, rot=<0 0 0 -1>}
 deform_elbow_out_r {parent=deform_elbow_out_base_r, loc=<0 0 0>, rot=<0 0 0 1>}
 upperarm_twist_01_r {parent=upperarm_r, loc=<-0.0041 0 0>, rot=<0 0 0 -1>}
 deform_upperarm_bk_base_r {parent=upperarm_twist_01_r, loc=<-1.0814 7.2192 1.9455>, rot=<0 0 0 1>}
 deform_upperarm_bk_r {parent=deform_upperarm_bk_base_r, loc=<0 0 0>, rot=<0 0 0 1>}
 deform_upperarm_fr_base_r {parent=upperarm_twist_01_r, loc=<0.3689 -7.0506 2.7652>, rot=<0 0 0 1>}
 deform_upperarm_fr_r {parent=deform_upperarm_fr_base_r, loc=<0 0 0>, rot=<0 0 0 1>}
 deform_upperarm_up_base_r {parent=upperarm_twist_01_r, loc=<3.2214 1.1391 -3.3308>, rot=<0 0 0 1>}
 deform_upperarm_up_r {parent=deform_upperarm_up_base_r, loc=<0 0 0>, rot=<0 0 0 -1>}
 upperarm_twist_02_r {parent=upperarm_r, loc=<-13.7941 0 0>, rot=<0 0 0 -1>}
 bag_upper_skin {parent=clavicle_r, loc=<-7.8994 -0.3993 -0.5311>, rot=<0.9991 0.0252 0.0336 0.0008>}
 def_bandolier_root {parent=spine_05, loc=<12.0976 -1.1998 11.8312>, rot=<0.4795 0.5196 0.5196 0.4795>}
 def_bandolier_front_01 {parent=def_bandolier_root, loc=<5.5449 -0.4846 -1.8699>, rot=<0.8937 -0.0575 -0.4084 0.1767>}
 def_bandolier_front_02 {parent=def_bandolier_front_01, loc=<6.754 0 0>, rot=<0.0381 0.1139 0.0008 -0.9928>}
 def_bandolier_front_03 {parent=def_bandolier_front_02, loc=<9.6779 0 0>, rot=<-0.0648 0.1338 -0.0594 -0.9871>}
 def_bandolier_front_04 {parent=def_bandolier_front_03, loc=<6.0729 0 0>, rot=<-0.0288 0.0779 -0.0416 -0.9957>}
 def_bandolier_front_05 {parent=def_bandolier_front_04, loc=<7.3332 0 0>, rot=<0.0055 0.0607 -0.0451 -0.9971>}
 def_bandolier_front_06 {parent=def_bandolier_front_05, loc=<4.5763 0 0>, rot=<-0.0107 0.0492 -0.045 -0.9977>}
 def_bandolier_front_07 {parent=def_bandolier_front_06, loc=<7.6663 0 0>, rot=<-0.0354 0.0624 -0.0584 -0.9957>}
 def_bandolier_front_08 {parent=def_bandolier_front_07, loc=<5.9931 0 0>, rot=<-0.0142 0 0 -0.9999>}
 def_bandolier_front_09 {parent=def_bandolier_front_08, loc=<3.6757 -0.7239 -0.5811>, rot=<0.1424 -0.0887 0.0692 -0.9834>}
 def_bandolier_front_10 {parent=def_bandolier_front_09, loc=<2.8099 0.4505 0.6633>, rot=<0.0291 -0.0804 0.1093 -0.9903>}
 def_bandolier_rear_01 {parent=def_bandolier_root, loc=<-4.7502 -0.7977 -1.3657>, rot=<-0.4491 0.2012 0.8628 -0.1159>}
 def_bandolier_rear_02 {parent=def_bandolier_rear_01, loc=<5.1157 0 0>, rot=<-0.0255 -0.1701 -0.0044 -0.9851>}
 def_bandolier_rear_03 {parent=def_bandolier_rear_02, loc=<6.1597 0 0>, rot=<0.0129 -0.1161 0.0015 -0.9932>}
 def_bandolier_rear_04 {parent=def_bandolier_rear_03, loc=<7.2069 0 0>, rot=<-0.0055 -0.0834 -0.0005 -0.9965>}
 def_bandolier_rear_05 {parent=def_bandolier_rear_04, loc=<7.1672 0 0>, rot=<-0.0076 -0.0269 -0.0173 -0.9995>}
 def_bandolier_rear_06 {parent=def_bandolier_rear_05, loc=<7.2767 0 0>, rot=<0.0199 0.018 -0.0073 -0.9996>}
 def_bandolier_rear_07 {parent=def_bandolier_rear_06, loc=<7.8233 0 0>, rot=<-0.0533 0.0035 -0.0653 -0.9964>}
 def_bandolier_rear_08 {parent=def_bandolier_rear_07, loc=<6.2254 0 0>, rot=<0 0 0 1>}
 def_bandolier_rear_09 {parent=def_bandolier_rear_08, loc=<2.5857 0.2277 -0.3569>, rot=<0.1068 -0.0284 0.003 -0.9939>}
 def_bandolier_rear_10 {parent=def_bandolier_rear_09, loc=<3.6407 2.1523 -0.4836>, rot=<-0.0091 0.0786 0.1149 -0.9902>}
 def_bandolier_root_Skin {parent=def_bandolier_root, loc=<0 0 0>, rot=<0 0 0 1>}
 base_scarf_sd_l {parent=spine_05, loc=<12.4807 6.1292 -5.5636>, rot=<0.1672 -0.6871 -0.2657 -0.6553>}
 dyn_scarf_sd_l {parent=base_scarf_sd_l, loc=<0 0 0>, rot=<0 0 0 1>}
 dyn_scarf_sd_extra {parent=base_scarf_sd_l, loc=<0 3.7853 0.6962>, rot=<-0.071 0 0 -0.9975>}
 base_scarf_sd_r {parent=spine_05, loc=<12.481 6.1292 5.5636>, rot=<0.6871 0.1672 -0.6553 0.2657>}
 dyn_scarf_sd_r {parent=base_scarf_sd_r, loc=<0 0 0>, rot=<0 0 0 -1>}
 base_scarf_bk_low_c {parent=spine_05, loc=<12.8843 -5.0011 0>, rot=<0.5196 -0.4795 0.4795 -0.5196>}
 dyn_scarf_bk_low_c {parent=base_scarf_bk_low_c, loc=<0 0 0>, rot=<0 0 0 1>}
 base_scarf_bk_c {parent=dyn_scarf_bk_low_c, loc=<-0.7548 0 2.8507>, rot=<0 0 0 1>}
 dyn_scarf_bk_c {parent=base_scarf_bk_c, loc=<0 0 0>, rot=<0 0 0 1>}
 base_scarf_fr_c {parent=spine_05, loc=<8.254 11.4304 -0.6091>, rot=<-0.1371 -0.9906 0 0>}
 dyn_scarf_fr_c {parent=base_scarf_fr_c, loc=<3.3113 0 0>, rot=<0 0 0 1>}
 clavicle_ASW_fix {parent=spine_05, loc=<7.0116 -6.1472 10.2901>, rot=<0.5196 -0.4795 0.4795 -0.5196>}
 deform_pec_aim_base_l {parent=spine_04, loc=<7.8732 12.6388 -2.4123>, rot=<0.0102 -0.4602 0.3979 -0.7936>}
 deform_pec_aim_l {parent=deform_pec_aim_base_l, loc=<0 0 0>, rot=<0 0 0 1>}
 deform_pec_base_l {parent=deform_pec_aim_l, loc=<10.766 0 0>, rot=<0 0 0 1>}
 deform_pec_l {parent=deform_pec_base_l, loc=<0.055 0 0>, rot=<0 0 0 1>}
 deform_pec_aim_base_r {parent=spine_04, loc=<7.8727 12.6388 2.4123>, rot=<-0.4602 -0.0102 0.7936 0.3979>}
 deform_pec_aim_r {parent=deform_pec_aim_base_r, loc=<0 0 0>, rot=<0 0 0 1>}
 deform_pec_base_r {parent=deform_pec_aim_r, loc=<-10.7662 -0.0006 -0.0007>, rot=<0 0 0 1>}
 deform_pec_r {parent=deform_pec_base_r, loc=<0 0 0>, rot=<0 0 0 1>}
 dyn_bag_chain_01 {parent=spine_03, loc=<1.7684 -0.0341 -1.7354>, rot=<0.0512 -0.9348 -0.0192 -0.3509>}
 dyn_bag_chain_02 {parent=dyn_bag_chain_01, loc=<9.8154 0 0>, rot=<0 0 0 1>}
 base_bag {parent=dyn_bag_chain_02, loc=<9.4169 0.3461 1.7162>, rot=<0.0042 -0.2542 0.0161 -0.967>}
 dyn_bag {parent=base_bag, loc=<0 0 0>, rot=<0 0 0 1>}
 back_fix {parent=spine_03, loc=<13.6427 -11.1216 0>, rot=<0.5266 -0.4719 0.4719 -0.5266>}
 thigh_l {parent=pelvis, loc=<-2.5693 -0.0109 -7.7501>, rot=<0.0012 0.0572 0.0155 -0.9982>}
 calf_l {parent=thigh_l, loc=<-39.0484 0 0>, rot=<0.0006 -0.0161 -0.0391 -0.9991>}
 foot_l {parent=calf_l, loc=<-42.561 0.1077 0.4292>, rot=<0.0418 -0.0455 0.0343 -0.9975>}
 ball_l {parent=foot_l, loc=<-6.1599 13.8264 0.6806>, rot=<0.0001 -0.0001 -0.7063 -0.7079>}
 calf_twist_02_l {parent=calf_l, loc=<-18.8064 -0.0967 0.0397>, rot=<0.0237 0.0095 -0.0035 -0.9997>}
 calf_twist_01_l {parent=calf_l, loc=<-35.813 -0.0967 0.0397>, rot=<0.0237 0.0095 -0.0035 -0.9997>}
 deform_knee_in_base_l {parent=calf_l, loc=<1.298 -5.2957 -0.0496>, rot=<0 0 0 -1>}
 deform_knee_in_l {parent=deform_knee_in_base_l, loc=<0 0 0>, rot=<0 0 0 1>}
 deform_knee_out_base_l {parent=calf_l, loc=<0 4 0>, rot=<0 0 0 -1>}
 deform_knee_out_l {parent=deform_knee_out_base_l, loc=<0 0 0>, rot=<0 0 0 1>}
 thigh_twist_01_l {parent=thigh_l, loc=<-21.6303 0 0>, rot=<0.0013 0 0 -1>}
 deform_groin_base_l {parent=thigh_l, loc=<-2.1908 8.1142 -1.7442>, rot=<0 0 0 -1>}
 deform_groin_l {parent=deform_groin_base_l, loc=<0 0 0>, rot=<0 0 0 1>}
 thigh_r {parent=pelvis, loc=<-2.5693 -0.0109 7.7501>, rot=<0.0572 -0.0012 0.9982 0.0155>}
 calf_r {parent=thigh_r, loc=<39.0484 0 0>, rot=<0.0006 -0.0161 -0.0391 -0.9991>}
 foot_r {parent=calf_r, loc=<42.561 -0.1077 -0.4292>, rot=<0.0418 -0.0455 0.0343 -0.9975>}
 ball_r {parent=foot_r, loc=<6.1599 -13.8264 -0.6806>, rot=<0.0001 -0.0001 -0.7063 -0.7079>}
 calf_twist_02_r {parent=calf_r, loc=<18.8064 0.0967 -0.0397>, rot=<0.0237 0.0095 -0.0035 -0.9997>}
 calf_twist_01_r {parent=calf_r, loc=<35.813 0.0967 -0.0397>, rot=<0.0237 0.0095 -0.0035 -0.9997>}
 deform_knee_in_base_r {parent=calf_r, loc=<-1.298 5.2957 0.0496>, rot=<0 0 0 -1>}
 deform_knee_in_r {parent=deform_knee_in_base_r, loc=<0 0 0>, rot=<0 0 0 1>}
 deform_knee_out_base_r {parent=calf_r, loc=<0 -4 0>, rot=<0 0 0 -1>}
 deform_knee_out_r {parent=deform_knee_out_base_r, loc=<0 0 0>, rot=<0 0 0 1>}
 thigh_twist_01_r {parent=thigh_r, loc=<21.6303 0 0>, rot=<0.0013 0 0 -1>}
 deform_groin_base_r {parent=thigh_r, loc=<2.1908 -8.1142 1.7442>, rot=<0 0 0 -1>}
 deform_groin_r {parent=deform_groin_base_r, loc=<0 0 0>, rot=<0 0 0 1>}
 deform_glute_base_l {parent=pelvis, loc=<-8.8064 -8.8222 -8.4287>, rot=<0.0275 0.0501 -0.4472 -0.8926>}
 deform_glute_l {parent=deform_glute_base_l, loc=<0 0 0>, rot=<0 0 0 1>}
 deform_glute_base_r {parent=pelvis, loc=<-8.8062 -8.8222 8.4287>, rot=<0.0501 -0.0275 0.8926 -0.4472>}
 deform_glute_r {parent=deform_glute_base_r, loc=<0 0 0>, rot=<0 0 0 1>}
 belt_area_l {parent=pelvis, loc=<-2.5693 -0.0109 -7.7501>, rot=<-0.5089 -0.491 -0.491 -0.5089>}
 base_fr_belt_l {parent=belt_area_l, loc=<8.9995 -0.4773 4.6315>, rot=<0 0 0 1>}
 dyn_fr_belt_l {parent=base_fr_belt_l, loc=<0 0 0>, rot=<0 0 0 1>}
 dyn_fr_belt_extra_l {parent=base_fr_belt_l, loc=<-1.0592 -1.6121 -3.5668>, rot=<0 0 0.2063 -0.9785>}
 base_sd_belt_l {parent=belt_area_l, loc=<0.1524 -6.7584 7.1695>, rot=<0 0 0.7071 -0.7071>}
 dyn_sd_belt_l {parent=base_sd_belt_l, loc=<0 0 0>, rot=<0 0 0 1>}
 base_bk_belt_l {parent=belt_area_l, loc=<-12.1197 1.8332 8.0041>, rot=<0 0 1 0>}
 dyn_bk_belt_l {parent=base_bk_belt_l, loc=<0 0 0>, rot=<0 0 0 1>}
 belt_area_r {parent=pelvis, loc=<-2.5693 -0.0109 7.7501>, rot=<-0.5089 -0.491 -0.491 -0.5089>}
 base_fr_belt_r {parent=belt_area_r, loc=<8.5496 1.1177 1.3657>, rot=<0 0 0 1>}
 dyn_fr_belt_r {parent=base_fr_belt_r, loc=<0 0 0>, rot=<0 0 0 1>}
 base_sd_belt_r {parent=belt_area_r, loc=<2.5914 7.9225 1.5682>, rot=<0 0 -0.7071 -0.7071>}
 dyn_sd_belt_r {parent=base_sd_belt_r, loc=<0 0 0>, rot=<0 0 0 1>}
 base_bk_belt_r {parent=belt_area_r, loc=<-12.1197 0.1879 5.8939>, rot=<0 0 1 0>}
 dyn_bk_belt_r {parent=base_bk_belt_r, loc=<0 0 0>, rot=<0 0 0 1>}
 base_fr_belt_c {parent=pelvis, loc=<0.5483 9.7475 0>, rot=<-0.5089 -0.491 -0.491 -0.5089>}
 dyn_fr_belt_c {parent=base_fr_belt_c, loc=<0 0 0>, rot=<0 0 0 1>}
 base_fr_shirt_lower_bagginess_l {parent=pelvis, loc=<10.9733 7.8641 -7.0722>, rot=<-0.4129 -0.5741 -0.392 -0.5885>}
 dyn_fr_shirt_lower_bagginess_l {parent=base_fr_shirt_lower_bagginess_l, loc=<0 0 0>, rot=<0 0 0 1>}
 base_sd_shirt_lower_bagginess_l {parent=pelvis, loc=<12.877 -2.619 -11.5256>, rot=<0.0277 -0.7066 0.053 -0.7051>}
 dyn_sd_shirt_lower_bagginess_l {parent=base_sd_shirt_lower_bagginess_l, loc=<0 0 0>, rot=<0 0 0 1>}
 base_fr_shirt_lower_bagginess_r {parent=pelvis, loc=<10.9172 6.298 10.7069>, rot=<0.6499 0.2786 0.6395 0.3017>}
 dyn_fr_shirt_lower_bagginess_r {parent=base_fr_shirt_lower_bagginess_r, loc=<0 0 0>, rot=<0 0 0 1>}
 base_bk_shirt_lower_bagginess_r {parent=pelvis, loc=<11.9804 -8.8863 6.3552>, rot=<0.5243 -0.4744 0.541 -0.4554>}
 dyn_bk_shirt_lower_bagginess_r {parent=base_bk_shirt_lower_bagginess_r, loc=<0 0 0>, rot=<0 0 0 1>}
 base_sd_shirt_lower_bagginess_r {parent=pelvis, loc=<12.1452 -4.2892 12.0374>, rot=<0.707 -0.0127 0.707 0.0127>}
 dyn_sd_shirt_lower_bagginess_r {parent=base_sd_shirt_lower_bagginess_r, loc=<0 0 0>, rot=<0 0 0 1>}
 base_bk_shirt_lower_bagginess_l {parent=pelvis, loc=<11.9552 -9.589 -5.7181>, rot=<0.5243 -0.4744 0.541 -0.4554>}
 dyn_bk_shirt_lower_bagginess_l {parent=base_bk_shirt_lower_bagginess_l, loc=<0 0 0>, rot=<0 0 0 1>}
 thigh_l_ctrl {controlType=EULER_TRANSFORM, type=Control, loc=<0 0 0>, rot=<0 0 0 1>}
 thigh_r_ctrl {controlType=EULER_TRANSFORM, type=Control, loc=<0 0 0>, rot=<0 0 0 1>}
 clavicle_l_ctrl {controlType=EULER_TRANSFORM, type=Control, loc=<0 0 0>, rot=<0 0 0 1>}
 upperarm_l_ctrl {controlType=EULER_TRANSFORM, parent=clavicle_l_ctrl, type=Control, loc=<0 0 0>, rot=<0 0 0 1>}
 clavicle_r_ctrl {controlType=EULER_TRANSFORM, type=Control, loc=<0 0 0>, rot=<0 0 0 1>}
 upperarm_r_ctrl {controlType=EULER_TRANSFORM, parent=clavicle_r_ctrl, type=Control, loc=<0 0 0>, rot=<0 0 0 1>}
 base_bag_null {type=Null, loc=<15.6816 -1.7325 101.5254>, rot=<-0.0105 0.6314 -0.0129 0.7753>}
 base_bag_ctrl {controlType=EULER_TRANSFORM, parent=base_bag_null, type=Control, loc=<0 0 0>, rot=<0 0 0 1>}
 def_bandolier_front_01_null {type=Null, loc=<-11.3466 -0.19 141.2516>, rot=<0.6726 0.5913 -0.1639 0.4137>}
 def_bandolier_front_01_ctrl {controlType=EULER_TRANSFORM, parent=def_bandolier_front_01_null, type=Control, loc=<0 0 0>, rot=<0 0 0 1>}
 def_bandolier_front_02_null {type=Null, loc=<-9.6782 4.2664 136.4584>, rot=<-0.6328 -0.5467 0.2171 -0.5035>}
 def_bandolier_front_02_ctrl {controlType=EULER_TRANSFORM, parent=def_bandolier_front_02_null, type=Control, loc=<0 0 0>, rot=<0 0 0 1>}
 def_bandolier_front_03_null {type=Null, loc=<-6.698 8.8459 128.4705>, rot=<0.6607 0.4206 -0.3046 0.5421>}
 def_bandolier_front_03_ctrl {controlType=EULER_TRANSFORM, parent=def_bandolier_front_03_null, type=Control, loc=<0 0 0>, rot=<0 0 0 1>}
 def_bandolier_front_04_null {type=Null, loc=<-3.9006 10.2159 123.2573>, rot=<-0.6672 -0.3404 0.3443 -0.5662>}
 def_bandolier_front_04_ctrl {controlType=EULER_TRANSFORM, parent=def_bandolier_front_04_null, type=Control, loc=<0 0 0>, rot=<0 0 0 1>}
 def_bandolier_front_05_null {type=Null, loc=<-0.0049 10.6874 117.0624>, rot=<0.6566 0.2768 -0.3564 0.6044>}
 def_bandolier_front_05_ctrl {controlType=EULER_TRANSFORM, parent=def_bandolier_front_05_null, type=Control, loc=<0 0 0>, rot=<0 0 0 1>}
 def_bandolier_front_06_null {type=Null, loc=<2.7072 10.3791 113.3893>, rot=<-0.6565 -0.2131 0.3637 -0.6256>}
 def_bandolier_front_06_ctrl {controlType=EULER_TRANSFORM, parent=def_bandolier_front_06_null, type=Control, loc=<0 0 0>, rot=<0 0 0 1>}
 def_bandolier_front_07_null {type=Null, loc=<7.6498 9.0352 107.6851>, rot=<0.6655 0.1219 -0.3741 0.6343>}
 def_bandolier_front_07_ctrl {controlType=EULER_TRANSFORM, parent=def_bandolier_front_07_null, type=Control, loc=<0 0 0>, rot=<0 0 0 1>}
 def_bandolier_front_08_null {type=Null, loc=<11.7878 7.1636 103.7747>, rot=<-0.6745 -0.1165 0.3758 -0.6247>}
 def_bandolier_front_08_ctrl {controlType=EULER_TRANSFORM, parent=def_bandolier_front_08_null, type=Control, loc=<0 0 0>, rot=<0 0 0 1>}
 def_bandolier_root_null {type=Null, loc=<-11.8312 -5.7349 143.1215>, rot=<0 0 0.7071 0.7071>}
 def_bandolier_root_ctrl {controlType=EULER_TRANSFORM, parent=def_bandolier_root_null, type=Control, loc=<0 0 0>, rot=<0 0 0 1>}
 def_bandolier_rear_01_null {type=Null, loc=<-11.0335 -10.485 141.7558>, rot=<-0.4598 -0.1753 0.5281 -0.6921>}
 def_bandolier_rear_01_ctrl {controlType=EULER_TRANSFORM, parent=def_bandolier_rear_01_null, type=Control, loc=<0 0 0>, rot=<0 0 0 1>}
 def_bandolier_rear_02_null {type=Null, loc=<-9.0856 -13.3997 138.0301>, rot=<0.5612 0.2749 -0.4434 0.6425>}
 def_bandolier_rear_02_ctrl {controlType=EULER_TRANSFORM, parent=def_bandolier_rear_02_null, type=Control, loc=<0 0 0>, rot=<0 0 0 1>}
 def_bandolier_rear_03_null {type=Null, loc=<-6.2793 -15.009 132.7883>, rot=<-0.6001 -0.3542 0.3727 -0.6128>}
 def_bandolier_rear_03_ctrl {controlType=EULER_TRANSFORM, parent=def_bandolier_rear_03_null, type=Control, loc=<0 0 0>, rot=<0 0 0 1>}
 def_bandolier_rear_04_null {type=Null, loc=<-2.8824 -15.2373 126.4363>, rot=<0.6326 0.4017 -0.323 0.578>}
 def_bandolier_rear_04_ctrl {controlType=EULER_TRANSFORM, parent=def_bandolier_rear_04_null, type=Control, loc=<0 0 0>, rot=<0 0 0 1>}
 def_bandolier_rear_05_null {type=Null, loc=<0.476 -14.2709 120.1789>, rot=<-0.6523 -0.4036 0.2989 -0.5677>}
 def_bandolier_rear_05_ctrl {controlType=EULER_TRANSFORM, parent=def_bandolier_rear_05_null, type=Control, loc=<0 0 0>, rot=<0 0 0 1>}
 def_bandolier_rear_06_null {type=Null, loc=<4.0814 -12.9087 114.0067>, rot=<0.6383 0.3945 -0.2984 0.5899>}
 def_bandolier_rear_06_ctrl {controlType=EULER_TRANSFORM, parent=def_bandolier_rear_06_null, type=Control, loc=<0 0 0>, rot=<0 0 0 1>}
 def_bandolier_rear_07_null {type=Null, loc=<8.0771 -11.7231 107.3861>, rot=<-0.6921 -0.3334 0.282 -0.5747>}
 def_bandolier_rear_07_ctrl {controlType=EULER_TRANSFORM, parent=def_bandolier_rear_07_null, type=Control, loc=<0 0 0>, rot=<0 0 0 1>}
 def_bandolier_rear_08_null {type=Null, loc=<11.9282 -10.8675 102.5701>, rot=<0.6921 0.3334 -0.282 0.5747>}
 def_bandolier_rear_08_ctrl {controlType=EULER_TRANSFORM, parent=def_bandolier_rear_08_null, type=Control, loc=<0 0 0>, rot=<0 0 0 1>}
 clavicle_l_null {type=Null, loc=<2.3223 -4.0648 135.4907>, rot=<-0.0008 -0.0336 0.0252 -0.9991>}
 clavicle_r_null {type=Null, loc=<-2.3223 -4.0648 135.4907>, rot=<0.9991 0.0252 0.0336 -0.0008>}
 upperarm_l_null {type=Null, loc=<14.6647 13.0779 126.7259>, rot=<-0.1926 0.4251 0.4944 0.7334>}
 upperarm_r_null {type=Null, loc=<-17.7788 -4.8444 134.4501>, rot=<-0.9345 -0.014 -0.3555 -0.0134>}
 spine_05_null {type=Null, loc=<0 -3.5695 131.159>, rot=<0.0284 -0.7065 -0.0284 0.7065>}
 spine_05_ctrl {controlType=EULER_TRANSFORM, type=Control, loc=<0 0 0>, rot=<0 0 0 1>}
 Hide_Scarf {type=Curve, loc=<0 0 0>, rot=<0 0 0 1>}
]
variables [
 Back_Leg_R {type=double, default=0}
 Back_leg_L {type=double, default=0}
 Bag_Rot {type=double, default=0}
 Bandolier_Front_Ctrls {
  type="TArray<FRigElementKey>"
  default=((Type=Control,Name="def_bandolier_front_01_ctrl"),(Type=Control,Name="def_bandolier_front_02_ctrl"),(Type=Control,Name="def_bandolier_front_03_ctrl"),(Type=Control,Name="def_bandolier_front_04_ctrl"))
 }
 Bandolier_Front_CtrlsNoRBAN {
  type="TArray<FRigElementKey>"
  default="((Type=Control,Name=\"def_bandolier_front_01_ctrl\"),(Type=Control,Name=\"def_bandolier_front_02_ctrl\"),(Type=Control,Name=\"def_bandolier_front_03_ctrl\"),(Type=Control,Name=\"def_bandolier_front_04_ctrl\"),(Type=Control,Name=\"def_bandolier_front_05_ctrl\"),(Type ...(truncated)"
 }
 Bandolier_Front_Joints {
  type="TArray<FRigElementKey>"
  default=((Type=Bone,Name="def_bandolier_front_01"),(Type=Bone,Name="def_bandolier_front_02"),(Type=Bone,Name="def_bandolier_front_03"),(Type=Bone,Name="def_bandolier_front_04"))
 }
 Bandolier_Front_JointsNoRBAN {
  type="TArray<FRigElementKey>"
  default="((Type=Bone,Name=\"def_bandolier_front_01\"),(Type=Bone,Name=\"def_bandolier_front_02\"),(Type=Bone,Name=\"def_bandolier_front_03\"),(Type=Bone,Name=\"def_bandolier_front_04\"),(Type=Bone,Name=\"def_bandolier_front_05\"),(Type=Bone,Name=\"def_bandolier_front_06\"),(Ty ...(truncated)"
 }
 Bandolier_Rear_Ctrls {
  type="TArray<FRigElementKey>"
  default=((Type=Control,Name="def_bandolier_rear_01_ctrl"),(Type=Control,Name="def_bandolier_rear_02_ctrl"),(Type=Control,Name="def_bandolier_rear_03_ctrl"),(Type=Control,Name="def_bandolier_rear_04_ctrl"))
 }
 Bandolier_Rear_CtrlsNoRBAN {
  type="TArray<FRigElementKey>"
  default="((Type=Control,Name=\"def_bandolier_rear_01_ctrl\"),(Type=Control,Name=\"def_bandolier_rear_02_ctrl\"),(Type=Control,Name=\"def_bandolier_rear_03_ctrl\"),(Type=Control,Name=\"def_bandolier_rear_04_ctrl\"),(Type=Control,Name=\"def_bandolier_rear_05_ctrl\"),(Type=Cont ...(truncated)"
 }
 Bandolier_Rear_Joints {
  type="TArray<FRigElementKey>"
  default=((Type=Bone,Name="def_bandolier_rear_01"),(Type=Bone,Name="def_bandolier_rear_02"),(Type=Bone,Name="def_bandolier_rear_03"),(Type=Bone,Name="def_bandolier_rear_04"))
 }
 Bandolier_Rear_JointsNoRBAN {
  type="TArray<FRigElementKey>"
  default="((Type=Bone,Name=\"def_bandolier_rear_01\"),(Type=Bone,Name=\"def_bandolier_rear_02\"),(Type=Bone,Name=\"def_bandolier_rear_03\"),(Type=Bone,Name=\"def_bandolier_rear_04\"),(Type=Bone,Name=\"def_bandolier_rear_05\"),(Type=Bone,Name=\"def_bandolier_rear_06\"),(Type=Bon ...(truncated)"
 }
 Crouching {type=double, default=0}
 DBNO {type=double, default=0}
 Emotes {type=double, default=0}
 Front_Leg_L {type=double, default=0}
 Front_Leg_R {type=double, default=0}
 Head_Rot_Left {type=double, default=0}
 Head_Rot_Right {type=double, default=0}
 Head_Rot_Up {type=double, default=0}
 Head_Rot_Up_Crouch {type=double, default=0}
 Head_Rot_Up_Crouch_Extrem {type=double, default=0}
 Melee {type=double, default=0}
 Moving2D {type=double, default=0}
 RBAN Enable {type=bool, default=False}
 Selected_Pistol {type=double, default=0}
 Side_Leg_L {type=double, default=0}
 Side_Leg_R {type=double, default=0}
 Skydiving {type=double, default=0}
 SkydivingBool {type=bool, default=False}
 UpperArm L Up {type=double, default=0}
]
