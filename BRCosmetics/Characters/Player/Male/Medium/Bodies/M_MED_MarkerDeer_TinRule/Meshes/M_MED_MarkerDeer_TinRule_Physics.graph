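# Physics asset for the MarkerDeer TinRule medium male body cosmetic, setting up a bandolier and bag for physical motion.
# Capsules on pelvis, spine and left thigh form the body volume around the strap.
# Front and rear bandolier chains hang from spine_03 with twist-locked, narrow swing limits that generally widen downward, joining at their last links into a loop.
# The bag bone constrains to spine_03 and, with free rotation, to both chains' mid links.
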PhysicsAsset /BRCosmetics/Characters/Player/Male/Medium/Bodies/M_MED_MarkerDeer_TinRule/Meshes/M_MED_MarkerDeer_TinRule_Physics
bodies {
 def_bandolier_front_04 [{shapeType=Capsule, length=3, radius=3, shapeName=def_bandolier_front_04_capsule, center=<0 0 -1.3348>, rotation=<rot 90 0 0>}]
 def_bandolier_front_05 [{shapeType=Capsule, length=4, radius=4, shapeName=def_bandolier_front_05_capsule, center=<3 0 -3.5041>, rotation=<rot 90 0 0>}]
 def_bandolier_front_06 [{shapeType=Capsule, length=4, radius=4, shapeName=def_bandolier_front_06_capsule, center=<3 0 -3.6182>, rotation=<rot 90 0 0>}]
 def_bandolier_front_07 [{shapeType=Capsule, length=4, radius=4, shapeName=def_bandolier_front_07_capsule, center=<3 0 -3.905>, rotation=<rot 90 0 0>}]
 def_bandolier_front_08 [{shapeType=Capsule, length=4, radius=4, shapeName=def_bandolier_front_08_capsule, center=<3 0 -1.5196>, rotation=<rot 90 0 0>}]
 def_bandolier_front_09 [{shapeType=Capsule, length=2.005, radius=2, shapeName=def_bandolier_front_09_capsule, center=<2 0 0>, rotation=<rot 90 0 0>}]
 def_bandolier_front_10 [{shapeType=Capsule, length=10, radius=5, shapeName=def_bandolier_rear_10_capsule, center=<2.9663 7.3 7.7>, rotation=<rot -0.0001 -17.6301 0>}]
 def_bandolier_rear_04 [{shapeType=Capsule, length=2, radius=2.7, shapeName=def_bandolier_rear_04_capsule, rotation=<rot 90 0 0>}]
 def_bandolier_rear_05 [{shapeType=Capsule, length=4, radius=2.7, shapeName=def_bandolier_rear_05_capsule, center=<0.5 0 0>, rotation=<rot 90 0 0>}]
 def_bandolier_rear_06 [{shapeType=Capsule, length=4, radius=2.7, shapeName=def_bandolier_rear_06_capsule, center=<3 0 0>, rotation=<rot 90 0 0>}]
 def_bandolier_rear_07 [{shapeType=Capsule, length=4, radius=2.7, shapeName=def_bandolier_rear_07_capsule, center=<3 0 0>, rotation=<rot 90 0 0>}]
 def_bandolier_rear_08 [{shapeType=Capsule, length=6, radius=2.7, shapeName=def_bandolier_rear_08_capsule, center=<3 0 0>, rotation=<rot 90 0 0>}]
 def_bandolier_rear_09 [{shapeType=Capsule, length=3.5, radius=2.7, shapeName=def_bandolier_rear_09_capsule, center=<2 0 0>, rotation=<rot 90 0 0>}]
 def_bandolier_rear_10 [
  {shapeType=Capsule, length=2.5, radius=7.5, shapeName=def_bandolier_rear_10_capsule, center=<11.3541 0 -6.9>, rotation=<rot 0 -15 0>}
  {shapeType=Capsule, length=10, radius=4, shapeName=def_bandolier_rear_10_capsule, center=<6.677 -1.5141 -3.8333>, rotation=<rot 90 0 17.1501>}
 ]
 dyn_bag [{shapeType=Capsule, length=1, radius=1, shapeName=dyn_bag_capsule, rotation=<rot 0 90.7218 -57.0137>}]
 pelvis [{shapeType=Capsule, length=10, radius=9.5, shapeName=pelvis_capsule}]
 spine_01 [
  {shapeType=Capsule, length=9.5, radius=9.5, shapeName=spine_01_capsule, center=<0.843 0 -0.3195>}
  {shapeType=Capsule, length=7.5, radius=11, shapeName=spine_01_capsule, center=<-1.3867 -0.0486 0.3409>}
  {shapeType=Capsule, length=12, radius=9.5, shapeName=spine_01_capsule, center=<8.297 -0.8727 -2.6171>, rotation=<rot 20.872 -0.0001 0>}
 ]
 spine_02 [
  {
   shapeType=Capsule
   length=16.4333
   radius=6.0041
   shapeName=spine_02_capsule
   center=<-2.1918 16.5721 6.8259>
   rotation=<rot 1.2634 -90 47.7058>
  }
  {shapeType=Capsule, length=16.5, radius=6.5, shapeName=spine_02_capsule, center=<4.0873 -0.807 0>}
  {shapeType=Capsule, length=19, radius=7, shapeName=spine_02_capsule, center=<-0.2975 -13.325 3.5134>, rotation=<rot -50.2801 0 0>}
 ]
 spine_03 [
  {shapeType=Capsule, length=15.5668, radius=11, shapeName=spine_03_capsule, center=<4.2002 -0.279 0>}
  {shapeType=Capsule, length=18, radius=11, shapeName=spine_03_capsule, center=<11.6335 -1.1828 0>}
 ]
 thigh_l [{shapeType=Capsule, length=30.5, radius=9.5, shapeName=thigh_l_capsule, center=<-10.2829 0 0.1498>, rotation=<rot 90 0 0>}]
}
constraints [
 {
  bone1Name=spine_01
  bone2Name=pelvis
  swing1LimitDegrees=45
  swing1Motion=Free
  swing2LimitDegrees=45
  swing2Motion=Free
  twistLimitDegrees=45
  twistMotion=Free
 }
 {
  bone1Name=def_bandolier_front_05
  bone2Name=def_bandolier_front_04
  swing1LimitDegrees=2
  swing2LimitDegrees=2
  twistLimitDegrees=10
  twistMotion=Locked
 }
 {
  bone1Name=def_bandolier_front_06
  bone2Name=def_bandolier_front_05
  swing1LimitDegrees=2
  swing2LimitDegrees=2
  twistLimitDegrees=10
  twistMotion=Locked
 }
 {
  bone1Name=def_bandolier_front_07
  bone2Name=def_bandolier_front_06
  swing1LimitDegrees=2
  swing2LimitDegrees=2
  twistLimitDegrees=10
  twistMotion=Locked
 }
 {
  bone1Name=def_bandolier_front_08
  bone2Name=def_bandolier_front_07
  swing1LimitDegrees=15
  swing2LimitDegrees=15
  swing2Motion=Locked
  twistLimitDegrees=4
  twistMotion=Locked
 }
 {
  bone1Name=def_bandolier_rear_05
  bone2Name=def_bandolier_rear_04
  swing1LimitDegrees=3
  swing2LimitDegrees=2
  swing2Motion=Locked
  twistLimitDegrees=10
  twistMotion=Locked
 }
 {
  bone1Name=def_bandolier_rear_06
  bone2Name=def_bandolier_rear_05
  swing1LimitDegrees=3
  swing2LimitDegrees=2
  swing2Motion=Locked
  twistLimitDegrees=10
  twistMotion=Locked
 }
 {
  bone1Name=def_bandolier_rear_07
  bone2Name=def_bandolier_rear_06
  swing1LimitDegrees=5
  swing2LimitDegrees=2
  swing2Motion=Locked
  twistLimitDegrees=10
  twistMotion=Locked
 }
 {
  bone1Name=def_bandolier_rear_08
  bone2Name=def_bandolier_rear_07
  swing1LimitDegrees=15
  swing2LimitDegrees=20
  swing2Motion=Locked
  twistLimitDegrees=4
  twistMotion=Locked
 }
 {
  bone1Name=spine_02
  bone2Name=spine_01
  swing1LimitDegrees=45
  swing1Motion=Free
  swing2LimitDegrees=45
  swing2Motion=Free
  twistLimitDegrees=45
  twistMotion=Free
 }
 {
  bone1Name=dyn_bag
  bone2Name=spine_03
  swing1LimitDegrees=45
  swing1Motion=Locked
  swing2LimitDegrees=30
  twistLimitDegrees=45
  twistMotion=Locked
 }
 {
  bone1Name=def_bandolier_front_04
  bone2Name=spine_03
  swing1LimitDegrees=45
  swing1Motion=Free
  swing2LimitDegrees=45
  swing2Motion=Free
  twistLimitDegrees=45
  twistMotion=Free
 }
 {
  bone1Name=def_bandolier_rear_04
  bone2Name=spine_03
  swing1LimitDegrees=45
  swing1Motion=Free
  swing2LimitDegrees=45
  swing2Motion=Free
  twistLimitDegrees=45
  twistMotion=Free
 }
 {
  bone1Name=thigh_l
  bone2Name=pelvis
  swing1LimitDegrees=45
  swing1Motion=Free
  swing2LimitDegrees=45
  swing2Motion=Free
  twistLimitDegrees=45
  twistMotion=Free
 }
 {
  bone1Name=def_bandolier_rear_10
  bone2Name=def_bandolier_rear_09
  swing1LimitDegrees=25
  swing2LimitDegrees=10
  swing2Motion=Locked
  twistLimitDegrees=45
  twistMotion=Locked
 }
 {
  bone1Name=def_bandolier_front_09
  bone2Name=def_bandolier_front_08
  swing1LimitDegrees=15
  swing2LimitDegrees=30
  swing2Motion=Locked
  twistLimitDegrees=45
  twistMotion=Locked
 }
 {
  bone1Name=def_bandolier_rear_09
  bone2Name=def_bandolier_rear_08
  swing1LimitDegrees=15
  swing2LimitDegrees=10
  twistLimitDegrees=45
  twistMotion=Locked
 }
 {
  bone1Name=def_bandolier_front_10
  bone2Name=def_bandolier_front_09
  swing1LimitDegrees=25
  swing2LimitDegrees=45
  swing2Motion=Locked
  twistLimitDegrees=45
  twistMotion=Locked
 }
 {
  bone1Name=def_bandolier_front_10
  bone2Name=def_bandolier_rear_10
  swing1LimitDegrees=30
  swing2LimitDegrees=20
  swing2Motion=Locked
  twistLimitDegrees=10
  twistMotion=Locked
 }
 {
  bone1Name=spine_03
  bone2Name=spine_02
  swing1LimitDegrees=45
  swing1Motion=Free
  swing2LimitDegrees=45
  swing2Motion=Free
  twistLimitDegrees=45
  twistMotion=Free
 }
 {
  bone1Name=dyn_bag
  bone2Name=def_bandolier_rear_07
  swing1LimitDegrees=45
  swing1Motion=Free
  swing2LimitDegrees=45
  swing2Motion=Free
  twistLimitDegrees=45
  twistMotion=Free
 }
 {
  bone1Name=dyn_bag
  bone2Name=def_bandolier_front_07
  swing1LimitDegrees=45
  swing1Motion=Free
  swing2LimitDegrees=45
  swing2Motion=Free
  twistLimitDegrees=45
  twistMotion=Free
 }
]
