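# Post-process style cosmetic AnimBP for the MarkerDeer TinRule male body, animating its scarf, shirt, bag and bandolier.
# The parent mesh pose feeds a Control Rig, then Anim Dynamics for cloth bones and a spine-relative RigidBody with additive bone fix-ups.
# When bRigidBodyEnabled is off, a cheaper branch skips all simulation.
# The event graph derives correction alphas from movement, weapon and spine twist, and picks gravity and sim-space settings per state.
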AnimBlueprint /BRCosmetics/Characters/Player/Male/Medium/Bodies/M_MED_MarkerDeer_TinRule/Meshes/M_MED_MarkerDeer_TinRule_AnimBP
target_skeleton=&/M_MED_MarkerDeer_TinRule_Skeleton
graphs [
 AnimGraph {nodes [
   Anim Dynamics - Bone: dyn_bag_chain_02(Alpha=0.5, ExternalForce="0,0,0") {
    angularDampingOverride=1
    angularSpringConstant=20000
    bAngularSpring=true
    bOverrideAngularDamping=true
    bUseGravityOverride=true
    gravityScale=0.35
    simulationSpace=World
    alphaBoolBlend {blendOption=Linear}
    boundBone {boneName=dyn_bag_chain_02}
    componentAppliedLinearAccClamp=<2000 2000 2000>
    componentLinearAccScale=<1 1 1>
    componentLinearVelScale=<1 1 1>
    gravityOverride {z=-980}
    physicsBodyDefinitions [{
      boundBone {boneName=dyn_bag_chain_02}
      boxExtents=<30 30 30>
      constraintSetup {angularLimitsMax=<0 15 20>, angularLimitsMin=<0 -15 -20>, angularTarget=<1 0 0>}
      localJointOffset=<8 0 0>
     }]
    relativeSpaceBone {boneName=spine_01}
   }
   Output Pose(Result=Blend Poses by bool(BlendPose_0=Use cached pose 'SetupRB'(), BlendPose_1=Use cached pose 'SetupNoRB'(), bActiveValue=Get bRigidBodyEnabled()) {blendType=Linear}) {layerGroup=None}
   n1 = Control Rig!(
    Source=Copy Pose From Mesh() {bUseAttachedParent=true}
    Alpha
    Crouching:double=Get Alpha_Crouching() -> (Alpha_Crouching:double)
    RBAN Enable:bool=Get bRigidBodyEnabled()
    Skydiving:double=Get Alpha_Skydiving() -> (Alpha_Skydiving:double)
    Moving2D:double=Get Alpha_Moving2D() -> (Alpha_Moving2D:double)
    Emotes:double=Get Alpha_Emote() -> (Alpha_Emote:double)
    Selected_Pistol:double=Get Selected_Pistol_Invert() -> (Selected_Pistol_Invert:double)
    SkydivingBool:bool=Get bIsSkyDiving()
    Bag_Rot:double=Get Alpha_Bag_Rot() -> (Alpha_Bag_Rot:double)
    DBNO:double=Get Alpha_DBNO() -> (Alpha_DBNO:double)
    Melee:double=Get Alpha_DetectedArmBackR_InMelee_0() -> (Alpha_DetectedArmBackR_InMelee_0:double)
   ) -> (Pose) {
    alphaBoolBlend {blendOption=Linear}
    controlRigAssetReference {blueprintRigClass=&/M_MED_MarkerDeer_TinRule_CtrlRig.M_MED_MarkerDeer_TinRule_CtrlRig_C}
   }
   Save cached pose 'ControlRig Cache'(Pose=@n1) {cacheName=ControlRig Cache}
   n2 = Anim Dynamics - Bone: base_scarf_fr_c!(
    ComponentPose=Local To Component(LocalPose=Use cached pose 'ControlRig Cache'())
    Alpha=1
    GravityOverride:Vector="0,0,-980"
    ExternalForce="0,0,0"
   ) -> (Pose) {
    angularDampingOverride=1
    angularSpringConstant=25000
    bAngularSpring=true
    bOverrideAngularDamping=true
    bUseGravityOverride=true
    gravityScale=0.35
    simulationSpace=BoneRelative
    alphaBoolBlend {blendOption=Linear}
    boundBone {boneName=base_scarf_fr_c}
    componentAppliedLinearAccClamp=<2000 2000 2000>
    componentLinearAccScale=<4 4 4>
    componentLinearVelScale=<4 4 4>
    gravityOverride {z=-980}
    physicsBodyDefinitions [{
      boundBone {boneName=base_scarf_fr_c}
      boxExtents=<30 40 40>
      constraintSetup {angularLimitsMax=<5 20 10>, angularLimitsMin=<-5 -20 -3>, angularTarget=<1 0 0>}
      localJointOffset=<4.5 0 0>
     }]
    relativeSpaceBone {boneName=pelvis}
   }
   n3 = Anim Dynamics - Bone: dyn_scarf_bk_c!(ComponentPose=@n2, Alpha=1, GravityOverride:Vector="0,0,-980", ExternalForce="0,0,0") -> (Pose) {
    angularDampingOverride=0.999
    angularSpringConstant=20000
    bAngularSpring=true
    bOverrideAngularDamping=true
    bUseGravityOverride=true
    gravityScale=0.35
    simulationSpace=BoneRelative
    alphaBoolBlend {blendOption=Linear}
    boundBone {boneName=dyn_scarf_bk_c}
    componentAppliedLinearAccClamp=<2000 2000 2000>
    componentLinearAccScale=<4 4 4>
    componentLinearVelScale=<4 4 4>
    gravityOverride {z=-980}
    physicsBodyDefinitions [{
      boundBone {boneName=dyn_scarf_bk_c}
      boxExtents=<20 30 30>
      constraintSetup {angularLimitsMax=<5 15 0>, angularLimitsMin=<-5 -15 0>, angularTarget=<1 0 0>}
      localJointOffset=<4 0 0>
     }]
    relativeSpaceBone {boneName=pelvis}
   }
   n4 = Anim Dynamics - Bone: dyn_scarf_sd_r!(ComponentPose=@n3, Alpha=1, GravityOverride:Vector="0,0,-980", ExternalForce="0,0,0") -> (Pose) {
    angularDampingOverride=0.999
    angularSpringConstant=20000
    bAngularSpring=true
    bOverrideAngularDamping=true
    bUseGravityOverride=true
    gravityScale=0.35
    simulationSpace=BoneRelative
    alphaBoolBlend {blendOption=Linear}
    boundBone {boneName=dyn_scarf_sd_r}
    componentAppliedLinearAccClamp=<2000 2000 2000>
    componentLinearAccScale=<4 4 4>
    componentLinearVelScale=<4 4 4>
    gravityOverride {z=-980}
    physicsBodyDefinitions [{
      boundBone {boneName=dyn_scarf_sd_r}
      boxExtents=<20 30 30>
      constraintSetup {angularLimitsMax=<5 15 0>, angularLimitsMin=<-5 -15 0>, angularTarget=<1 0 0>}
      localJointOffset=<4 0 0>
     }]
    relativeSpaceBone {boneName=pelvis}
   }
   n5 = Anim Dynamics - Bone: dyn_scarf_sd_l!(ComponentPose=@n4, Alpha=1, GravityOverride:Vector="0,0,-980", ExternalForce="0,0,0") -> (Pose) {
    angularDampingOverride=0.999
    angularSpringConstant=20000
    bAngularSpring=true
    bOverrideAngularDamping=true
    bUseGravityOverride=true
    gravityScale=0.35
    simulationSpace=BoneRelative
    alphaBoolBlend {blendOption=Linear}
    boundBone {boneName=dyn_scarf_sd_l}
    componentAppliedLinearAccClamp=<2000 2000 2000>
    componentLinearAccScale=<4 4 4>
    componentLinearVelScale=<4 4 4>
    gravityOverride {z=-980}
    physicsBodyDefinitions [{
      boundBone {boneName=dyn_scarf_sd_l}
      boxExtents=<20 30 30>
      constraintSetup {angularLimitsMax=<5 15 0>, angularLimitsMin=<-5 -15 0>, angularTarget=<1 0 0>}
      localJointOffset=<4 0 0>
     }]
    relativeSpaceBone {boneName=pelvis}
   }
   n6 = Anim Dynamics - Bone: dyn_fr_shirt_lower_bagginess_l(ComponentPose=@n5, Alpha=1, ExternalForce="0,0,0") {
    angularDampingOverride=0.9999
    angularSpringConstant=8000
    bAngularSpring=true
    bOverrideAngularDamping=true
    bUseGravityOverride=true
    gravityScale=0.35
    linearSpringConstant=0.5
    simulationSpace=BoneRelative
    alphaBoolBlend {blendOption=Linear}
    boundBone {boneName=dyn_fr_shirt_lower_bagginess_l}
    componentAppliedLinearAccClamp=<3000 3000 3000>
    componentLinearAccScale=<0.4 0.4 0.4>
    componentLinearVelScale=<0.4 0.4 0.4>
    gravityOverride {z=-500}
    physicsBodyDefinitions [{
      boundBone {boneName=dyn_fr_shirt_lower_bagginess_l}
      boxExtents=<20 25 25>
      constraintSetup {angularLimitsMax=<15 40 20>, angularLimitsMin=<-15 -40 -20>, angularTarget=<1 0 0>, linearAxesMax=<0 1 5>}
      localJointOffset=<-5 0 0>
     }]
    relativeSpaceBone {boneName=root}
   }
   n7 = Anim Dynamics - Bone: dyn_sd_shirt_lower_bagginess_l(ComponentPose=@n6, Alpha=1, ExternalForce="0,0,0") {
    angularDampingOverride=0.9999
    angularSpringConstant=8000
    bAngularSpring=true
    bOverrideAngularDamping=true
    bUseGravityOverride=true
    gravityScale=0.35
    linearSpringConstant=0.5
    simulationSpace=BoneRelative
    alphaBoolBlend {blendOption=Linear}
    boundBone {boneName=dyn_sd_shirt_lower_bagginess_l}
    componentAppliedLinearAccClamp=<3000 3000 3000>
    componentLinearAccScale=<0.4 0.4 0.4>
    componentLinearVelScale=<0.4 0.4 0.4>
    gravityOverride {z=-500}
    physicsBodyDefinitions [{
      boundBone {boneName=dyn_sd_shirt_lower_bagginess_l}
      boxExtents=<20 25 25>
      constraintSetup {angularLimitsMax=<15 40 20>, angularLimitsMin=<-15 -40 -20>, angularTarget=<1 0 0>, linearAxesMax=<0 1 5>}
      localJointOffset=<-5 0 0>
     }]
    relativeSpaceBone {boneName=root}
   }
   n8 = Anim Dynamics - Bone: dyn_bk_shirt_lower_bagginess_l(ComponentPose=@n7, Alpha=1, ExternalForce="0,0,0") {
    angularDampingOverride=0.9999
    angularSpringConstant=8000
    bAngularSpring=true
    bOverrideAngularDamping=true
    bUseGravityOverride=true
    gravityScale=0.35
    linearDampingOverride=0.6
    linearSpringConstant=0.5
    simulationSpace=BoneRelative
    alphaBoolBlend {blendOption=Linear}
    boundBone {boneName=dyn_bk_shirt_lower_bagginess_l}
    componentAppliedLinearAccClamp=<3000 3000 3000>
    componentLinearAccScale=<0.4 0.4 0.4>
    componentLinearVelScale=<0.4 0.4 0.4>
    gravityOverride {z=-500}
    physicsBodyDefinitions [{
      boundBone {boneName=dyn_bk_shirt_lower_bagginess_l}
      boxExtents=<20 25 25>
      constraintSetup {angularLimitsMax=<15 40 20>, angularLimitsMin=<-15 -40 -20>, angularTarget=<1 0 0>, linearAxesMax=<0 1 5>}
      localJointOffset=<-5 0 0>
     }]
    relativeSpaceBone {boneName=root}
   }
   n9 = Anim Dynamics - Bone: dyn_fr_shirt_lower_bagginess_r(ComponentPose=@n8, Alpha=1, ExternalForce="0,0,0") {
    angularDampingOverride=0.9999
    angularSpringConstant=8000
    bAngularSpring=true
    bOverrideAngularDamping=true
    bUseGravityOverride=true
    gravityScale=0.35
    linearSpringConstant=0.5
    simulationSpace=BoneRelative
    alphaBoolBlend {blendOption=Linear}
    boundBone {boneName=dyn_fr_shirt_lower_bagginess_r}
    componentAppliedLinearAccClamp=<3000 3000 3000>
    componentLinearAccScale=<0.4 0.4 0.4>
    componentLinearVelScale=<0.4 0.4 0.4>
    gravityOverride {z=-500}
    physicsBodyDefinitions [{
      boundBone {boneName=dyn_fr_shirt_lower_bagginess_r}
      boxExtents=<20 25 25>
      constraintSetup {angularLimitsMax=<15 40 20>, angularLimitsMin=<-15 -40 -20>, angularTarget=<1 0 0>, linearAxesMax=<0 1 5>}
      localJointOffset=<5 0 0>
     }]
    relativeSpaceBone {boneName=root}
   }
   n10 = Anim Dynamics - Bone: dyn_sd_shirt_lower_bagginess_r(ComponentPose=@n9, Alpha=1, ExternalForce="0,0,0") {
    angularDampingOverride=0.9999
    angularSpringConstant=8000
    bAngularSpring=true
    bOverrideAngularDamping=true
    bUseGravityOverride=true
    gravityScale=0.35
    linearSpringConstant=0.5
    simulationSpace=BoneRelative
    alphaBoolBlend {blendOption=Linear}
    boundBone {boneName=dyn_sd_shirt_lower_bagginess_r}
    componentAppliedLinearAccClamp=<3000 3000 3000>
    componentLinearAccScale=<0.4 0.4 0.4>
    componentLinearVelScale=<0.4 0.4 0.4>
    gravityOverride {z=-500}
    physicsBodyDefinitions [{
      boundBone {boneName=dyn_sd_shirt_lower_bagginess_r}
      boxExtents=<20 25 25>
      constraintSetup {angularLimitsMax=<15 40 20>, angularLimitsMin=<-15 -40 -20>, angularTarget=<1 0 0>, linearAxesMax=<0 1 5>}
      localJointOffset=<5 0 0>
     }]
    relativeSpaceBone {boneName=root}
   }
   n11 = Anim Dynamics - Bone: dyn_bk_shirt_lower_bagginess_r(ComponentPose=@n10, Alpha=1, ExternalForce="0,0,0") {
    angularDampingOverride=0.9999
    angularSpringConstant=8000
    bAngularSpring=true
    bOverrideAngularDamping=true
    bUseGravityOverride=true
    gravityScale=0.35
    linearDampingOverride=0.6
    linearSpringConstant=0.5
    simulationSpace=BoneRelative
    alphaBoolBlend {blendOption=Linear}
    boundBone {boneName=dyn_bk_shirt_lower_bagginess_r}
    componentAppliedLinearAccClamp=<3000 3000 3000>
    componentLinearAccScale=<0.4 0.4 0.4>
    componentLinearVelScale=<0.4 0.4 0.4>
    gravityOverride {z=-500}
    physicsBodyDefinitions [{
      boundBone {boneName=dyn_bk_shirt_lower_bagginess_r}
      boxExtents=<20 25 25>
      constraintSetup {angularLimitsMax=<15 40 20>, angularLimitsMin=<-15 -40 -20>, angularTarget=<1 0 0>, linearAxesMax=<0 1 5>}
      localJointOffset=<5 0 0>
     }]
    relativeSpaceBone {boneName=root}
   }
   Save cached pose 'Dynamics Cache'(Pose=Component To Local(ComponentPose=@n11)) {cacheName=Dynamics Cache}
   n12 = "Transform (Modify) Bone - Bone: dyn_bag_chain_01"(
    ComponentPose=Local To Component(LocalPose=Use cached pose 'Dynamics Cache'())
    Alpha=1
    Translation="1.2,0,4"
    Rotation="10,0,0"
    Scale="1,1,1"
   ) {
    rotationMode=BMM_Additive
    rotationSpace=BCS_BoneSpace
    translationMode=BMM_Additive
    translationSpace=BCS_BoneSpace
    alphaBoolBlend {blendOption=Linear}
    alphaScaleBiasClamp {bInterpResult=true}
    boneToModify {boneName=dyn_bag_chain_01}
    rotation {pitch=10}
    translation {x=1.2, z=4}
   }
   n13 = "Transform (Modify) Bone - Bone: base_bag"(ComponentPose=@n12, Alpha=1, Translation="0,0,0", Rotation="40,0,0", Scale="1,1,1") {
    rotationMode=BMM_Additive
    rotationSpace=BCS_BoneSpace
    translationMode=BMM_Additive
    translationSpace=BCS_BoneSpace
    alphaBoolBlend {blendOption=Linear}
    alphaScaleBiasClamp {bInterpResult=true}
    boneToModify {boneName=base_bag}
    rotation {pitch=40}
   }
   n14 = "Transform (Modify) Bone - Bone: def_bandolier_rear_02"(ComponentPose=@n13, Alpha=1, Translation="0,0,0", Rotation="10,7,0", Scale="1,1,1") {
    rotationMode=BMM_Additive
    rotationSpace=BCS_BoneSpace
    translationMode=BMM_Additive
    translationSpace=BCS_BoneSpace
    alphaBoolBlend {blendInTime=0.2, blendOption=Linear, blendOutTime=0.2}
    alphaScaleBiasClamp {bInterpResult=true}
    boneToModify {boneName=def_bandolier_rear_02}
    rotation {pitch=10, yaw=7}
   }
   n15 = RigidBody!(
    ComponentPose=@n14
    Alpha
    OverrideWorldGravity:Vector="0,0,-980"
    ExternalForce
    ComponentLinearAccScale:Vector="0,0,0"
    ComponentLinearVelScale:Vector="0,0,0"
    SimSpaceSettings:Sim Space Settings Structure=(WorldAlpha=1,VelocityScaleZ=0.25,DampingAlpha=1,MaxLinearVelocity=100,MaxAngularVelocity=10,MaxLinearAcceleration=200,MaxAngularAcceleration=30,ExternalLinearDragV=(X=1,Y=1,Z=1),ExternalLinearVelocity=(X=0,Y=0,Z=0),ExternalAngularVelocity=(X=0,Y=0,Z=0))
   ) -> (Pose) {
    bOverrideWorldGravity=true
    overridePhysicsAsset=&/M_MED_MarkerDeer_TinRule_Physics
    simulationSpace=BaseBoneSpace
    alphaBoolBlend {blendOption=Linear}
    baseBoneRef {boneName=spine_01}
    componentAppliedLinearAccClamp=<3000 3000 3000>
    overrideWorldGravity {z=-980}
    simSpaceSettings {
     maxAngularAcceleration=30
     maxAngularVelocity=10
     maxLinearAcceleration=200
     maxLinearVelocity=100
     velocityScaleZ=0.25
     worldAlpha=1
     externalLinearDragV=<1 1 1>
    }
   }
   n16 = "Transform (Modify) Bone - Bone: def_bandolier_front_08"!(ComponentPose=@n15, bAlphaBoolEnabled:bool=True, Translation="-2.653,0.000,3.164", Rotation="1.514,-32.378,3.653", Scale="1,1,1") -> (Pose) {
    alphaInputType=Bool
    rotationMode=BMM_Additive
    rotationSpace=BCS_BoneSpace
    translationMode=BMM_Additive
    translationSpace=BCS_BoneSpace
    alphaBoolBlend {blendInTime=0.4, blendOption=Linear, blendOutTime=0.2}
    alphaScaleBiasClamp {bInterpResult=true}
    boneToModify {boneName=def_bandolier_front_08}
    rotation=<rot 1.514 -32.378 3.653>
    translation {x=-2.653, z=3.164}
   }
   n17 = "Transform (Modify) Bone - Bone: def_bandolier_rear_08"!(ComponentPose=@n16, bAlphaBoolEnabled:bool=True, Translation="-1.019,-1.993,-0.715", Rotation="-3.003,-26.866,-1.703", Scale="1,1,1") -> (Pose) {
    alphaInputType=Bool
    rotationMode=BMM_Additive
    rotationSpace=BCS_BoneSpace
    translationMode=BMM_Additive
    translationSpace=BCS_BoneSpace
    alphaBoolBlend {blendInTime=0.4, blendOption=Linear, blendOutTime=0.2}
    alphaScaleBiasClamp {bInterpResult=true}
    boneToModify {boneName=def_bandolier_rear_08}
    rotation=<rot -3.003 -26.866 -1.703>
    translation=<-1.019 -1.993 -0.715>
   }
   n18 = "Transform (Modify) Bone - Bone: def_bandolier_front_08"(ComponentPose=@n17, bAlphaBoolEnabled=True, Translation="-0.998,0.565,-0.806", Rotation="-0.058,15.356,-0.092", Scale="1,1,1") {
    alphaInputType=Bool
    rotationMode=BMM_Additive
    rotationSpace=BCS_BoneSpace
    translationMode=BMM_Additive
    translationSpace=BCS_BoneSpace
    alphaBoolBlend {blendInTime=0.2, blendOption=Linear, blendOutTime=0.2}
    alphaScaleBiasClamp {bInterpResult=true}
    boneToModify {boneName=def_bandolier_front_08}
    rotation=<rot -0.058 15.356 -0.092>
    translation=<-0.998 0.565 -0.806>
   }
   n19 = "Transform (Modify) Bone - Bone: def_bandolier_rear_08"(ComponentPose=@n18, bAlphaBoolEnabled=True, Translation="-0.561,0.015,-1.488", Rotation="0.075,13.481,0.030", Scale="1,1,1") {
    alphaInputType=Bool
    rotationMode=BMM_Additive
    rotationSpace=BCS_BoneSpace
    translationMode=BMM_Additive
    translationSpace=BCS_BoneSpace
    alphaBoolBlend {blendInTime=0.2, blendOption=Linear, blendOutTime=0.2}
    alphaScaleBiasClamp {bInterpResult=true}
    boneToModify {boneName=def_bandolier_rear_08}
    rotation=<rot 0.075 13.481 0.03>
    translation=<-0.561 0.015 -1.488>
   }
   n20 = "Transform (Modify) Bone - Bone: base_scarf_bk_c"(ComponentPose=@n19, Alpha=1, Translation="0,0,-3", Rotation="0,0,0", Scale="1,1,1") {
    rotationMode=BMM_Additive
    rotationSpace=BCS_BoneSpace
    translationMode=BMM_Additive
    translationSpace=BCS_BoneSpace
    alphaBoolBlend {blendOption=Linear}
    alphaScaleBiasClamp {bInterpResult=true}
    boneToModify {boneName=base_scarf_bk_c}
    translation {z=-3}
   }
   n21 = "Transform (Modify) Bone - Bone: bag_upper_skin"!(ComponentPose=@n20, bAlphaBoolEnabled:bool=True, Translation="0,0,2", Rotation="0,0,0", Scale="1,1,1") -> (Pose) {
    alphaInputType=Bool
    rotationMode=BMM_Additive
    rotationSpace=BCS_BoneSpace
    translationMode=BMM_Additive
    translationSpace=BCS_BoneSpace
    alphaBoolBlend {blendInTime=0.2, blendOption=Linear, blendOutTime=0.2}
    alphaScaleBiasClamp {bInterpResult=true}
    boneToModify {boneName=bag_upper_skin}
    translation {z=2}
   }
   n22 = "Transform (Modify) Bone - Bone: def_bandolier_rear_01"!(ComponentPose=@n21, bAlphaBoolEnabled:bool=True, Translation="0,0,1.5", Rotation="0,0,0", Scale="1,1,1") -> (Pose) {
    alphaInputType=Bool
    rotationMode=BMM_Additive
    rotationSpace=BCS_BoneSpace
    translationMode=BMM_Additive
    translationSpace=BCS_BoneSpace
    alphaBoolBlend {blendInTime=0.2, blendOption=Linear, blendOutTime=0.2}
    alphaScaleBiasClamp {bInterpResult=true}
    boneToModify {boneName=def_bandolier_rear_01}
    translation {z=1.5}
   }
   n23 = "Transform (Modify) Bone - Bone: def_bandolier_front_01"!(ComponentPose=@n22, bAlphaBoolEnabled:bool=True, Translation="0,0,0", Rotation="-1.5,0,0", Scale="1,1,1") -> (Pose) {
    alphaInputType=Bool
    rotationMode=BMM_Additive
    rotationSpace=BCS_BoneSpace
    translationMode=BMM_Additive
    translationSpace=BCS_BoneSpace
    alphaBoolBlend {blendInTime=0.2, blendOption=Linear, blendOutTime=0.2}
    alphaScaleBiasClamp {bInterpResult=true}
    boneToModify {boneName=def_bandolier_front_01}
    rotation {pitch=-1.5}
   }
   n24 = "Transform (Modify) Bone - Bone: bag_upper_skin"!(ComponentPose=@n23, Alpha:float=1, Translation="-2,0,2", Rotation="-30,0,0", Scale="1,1,1") -> (Pose) {
    rotationMode=BMM_Additive
    rotationSpace=BCS_BoneSpace
    translationMode=BMM_Additive
    translationSpace=BCS_BoneSpace
    alphaBoolBlend {blendInTime=0.2, blendOption=Linear, blendOutTime=0.2}
    alphaScaleBiasClamp {bInterpResult=true}
    boneToModify {boneName=bag_upper_skin}
    rotation {pitch=-30}
    translation {x=-2, z=2}
   }
   n25 = "Transform (Modify) Bone - Bone: back_fix"(ComponentPose=@n24, Alpha=1, Translation="-2,0,0", Rotation="0,0,0", Scale="1,1,1") {
    rotationMode=BMM_Additive
    rotationSpace=BCS_BoneSpace
    translationMode=BMM_Additive
    translationSpace=BCS_BoneSpace
    alphaBoolBlend {blendInTime=0.2, blendOption=Linear, blendOutTime=0.2}
    alphaScaleBiasClamp {bInterpResult=true}
    boneToModify {boneName=back_fix}
    translation {x=-2}
   }
   n26 = "Transform (Modify) Bone - Bone: back_fix"!(ComponentPose=@n25, bAlphaBoolEnabled:bool=True, Translation="-2,0,0", Rotation="0,0,0", Scale="1,1,1") -> (Pose) {
    alphaInputType=Bool
    rotationMode=BMM_Additive
    rotationSpace=BCS_BoneSpace
    translationMode=BMM_Additive
    translationSpace=BCS_BoneSpace
    alphaBoolBlend {blendInTime=0.2, blendOption=Linear, blendOutTime=0.2}
    alphaScaleBiasClamp {bInterpResult=true}
    boneToModify {boneName=back_fix}
    translation {x=-2}
   }
   n27 = "Transform (Modify) Bone - Bone: attach_cape"!(ComponentPose=@n26, bAlphaBoolEnabled:bool=True, Translation="0,-2,5", Rotation="0,0,10", Scale="1,1,1") -> (Pose) {
    alphaInputType=Bool
    rotationMode=BMM_Additive
    rotationSpace=BCS_BoneSpace
    translationMode=BMM_Additive
    translationSpace=BCS_BoneSpace
    alphaBoolBlend {blendInTime=0.2, blendOption=Linear, blendOutTime=1}
    alphaScaleBiasClamp {bInterpResult=true}
    boneToModify {boneName=attach_cape}
    rotation {roll=10}
    translation {y=-2, z=5}
   }
   n28 = "Transform (Modify) Bone - Bone: def_bandolier_rear_07"!(ComponentPose=@n27, bAlphaBoolEnabled:bool=True, Translation="0,0,-2", Rotation="0,0,0", Scale="1,1,1") -> (Pose) {
    alphaInputType=Bool
    rotationMode=BMM_Additive
    rotationSpace=BCS_BoneSpace
    translationMode=BMM_Additive
    translationSpace=BCS_BoneSpace
    alphaBoolBlend {blendInTime=0.2, blendOption=Linear, blendOutTime=0.2}
    alphaScaleBiasClamp {bInterpResult=true}
    boneToModify {boneName=def_bandolier_rear_07}
    translation {z=-2}
   }
   n29 = "Transform (Modify) Bone - Bone: def_bandolier_front_07"!(ComponentPose=@n28, bAlphaBoolEnabled:bool=True, Translation="0,0,-1", Rotation="0,0,0", Scale="1,1,1") -> (Pose) {
    alphaInputType=Bool
    rotationMode=BMM_Additive
    rotationSpace=BCS_BoneSpace
    translationMode=BMM_Additive
    translationSpace=BCS_BoneSpace
    alphaBoolBlend {blendInTime=0.2, blendOption=Linear, blendOutTime=0.2}
    alphaScaleBiasClamp {bInterpResult=true}
    boneToModify {boneName=def_bandolier_front_07}
    translation {z=-1}
   }
   Save cached pose 'SetupRB'(Pose=Component To Local(ComponentPose=@n29)) {cacheName=SetupRB}
   n30 = "Transform (Modify) Bone - Bone: back_fix"(
    ComponentPose=Local To Component(LocalPose=Use cached pose 'ControlRig Cache'())
    bAlphaBoolEnabled=True
    Translation="-0.5,0,0"
    Rotation="10,0,0"
    Scale="1,1,1"
   ) {
    alphaInputType=Bool
    rotationMode=BMM_Additive
    rotationSpace=BCS_BoneSpace
    translationMode=BMM_Additive
    translationSpace=BCS_BoneSpace
    alphaBoolBlend {blendInTime=0.1, blendOption=Linear, blendOutTime=0.1}
    boneToModify {boneName=back_fix}
    rotation {pitch=10}
    translation {x=-0.5}
   }
   n31 = "Transform (Modify) Bone - Bone: clavicle_ASW_fix"(ComponentPose=@n30, Alpha=1, Translation="-1.5,0,-1.5", Rotation="0,0,0", Scale="1,1,1") {
    rotationMode=BMM_Additive
    rotationSpace=BCS_BoneSpace
    translationMode=BMM_Additive
    translationSpace=BCS_BoneSpace
    alphaBoolBlend {blendInTime=0.1, blendOption=Linear, blendOutTime=0.1}
    alphaScaleBiasClamp {bInterpResult=true}
    boneToModify {boneName=clavicle_ASW_fix}
    translation {x=-1.5, z=-1.5}
   }
   Save cached pose 'SetupNoRB'(Pose=Component To Local(ComponentPose=@n31)) {cacheName=SetupNoRB}
  ]}
 EventGraph {nodes [
   n1 = Event BlueprintUpdateAnimation()
   > Set DeltaTimeX(DeltaTimeX:double=@n1.DeltaTimeX) -> (Output_Get:double)
   > n2 = Sequence!(execute) -> (then_0, then_1, then_2:Exec, then_3:Exec, then_4:Exec, then_5:Exec, then_6:Exec)
   > n3 = Gravity Override State Function DEPRECATED(
    CharacterPartAnimInstance=Self-Reference()
    GameTimeInSeconds=GetGameTimeInSeconds()
    GravityOverrideParams=Get Gravity Override Params()
    RBAN Params=Get RBAN Params()
   )
   > Set Gravity Override(Gravity Override=@n3)
   > n4 = Set RBAN Settings Out!(execute, RBAN Settings Out:RBANSetting Struct Structure=@n3.RBAN SettingsOut) -> (
    then
    Output_Get_ComponentLinearAccScale_28_#1:Vector=~
    Output_Get_ComponentLinearVelScale_31_#2:Vector=~
    Output_Get_SimSpaceSettings_34_#3:Sim Space Settings Structure
   )
   n5 = Get bIsSkyDiving()
   n6 = Get bIsPlayingEmote()
   n7 = Get bIsMoving2D()
   > n8 = Branch(Condition=AND Boolean!(A=Get bIsInFrontEnd(), B=NOT Boolean(A=@n7), C:bool=NOT Boolean(A=@n5), D:bool=NOT Boolean(A=@n6)) -> (ReturnValue))
   > Set SimSpaceSettings(SimSpaceSettings:Sim Space Settings Structure=Get SimSpaceSettingsFEIdle() -> (SimSpaceSettingsFEIdle:Sim Space Settings Structure)) -> (Output_Get:Sim Space Settings Structure)
   Set SimSpaceSettings(execute=@n8.else, SimSpaceSettings=@n4.Output_Get_SimSpaceSettings_34_#3)
   n9 = GetOwnerSkeletalMeshComponent()
   n10 = BreakRotator(InRot=BreakTransform(InTransform=GetDeltaTransformFromRefPose(self=@n9, BoneName=spine_05, BaseName=pelvis)).Rotation)
   n11 = Branch(
    execute=@n2.then_5
    Condition=AND Boolean!(A="float < float"(A=@n10, B=-30), B=Get bIsPlayingMeleeAnim(), C=Get Selected_Pickaxe() -> (Selected_Pickaxe:bool)) -> (ReturnValue)
   )
   > Set Alpha_DetectedArmBackR_InMelee_0(Alpha_DetectedArmBackR_InMelee_0:double=1) -> (Output_Get:double)
   Set Alpha_DetectedArmBackR_InMelee_0(execute=@n11.else, Alpha_DetectedArmBackR_InMelee_0=0)
   n12 = Get bIsCrouching()
   Set Alpha_Crouching(
    execute=@n2.then_1
    Alpha_Crouching:double=FInterpTo_Constant!(
     Current:double=Get Alpha_Crouching()
     Target:double="To Float (Boolean)"(InBool=@n12)
     DeltaTime:double=Get DeltaTimeX() -> (DeltaTimeX:double)
     InterpSpeed:double=5
    ) -> (ReturnValue:double)
   ) -> (Output_Get:double)
   n13 = Get bIsMoving2D()
   > Set Alpha_Moving2D(Alpha_Moving2D:double=FInterpTo_Constant(Current=Get Alpha_Moving2D(), Target="To Float (Boolean)"(InBool=@n13), DeltaTime=Get DeltaTimeX(), InterpSpeed=5)) -> (Output_Get:double)
   n14 = Get bIsDBNO()
   > Set Alpha_DBNO(Alpha_DBNO:double=FInterpTo_Constant(Current=Get Alpha_DBNO(), Target="To Float (Boolean)"(InBool=@n14), DeltaTime=Get DeltaTimeX(), InterpSpeed=5)) -> (Output_Get:double)
   > Set Alpha_Melee(Alpha_Melee:double="To Float (Boolean)"(InBool=Get bIsPlayingMeleeAnim())) -> (Output_Get:double)
   n15 = AND Boolean!(A=Get bIsTacticalSprinting(), B=Get bIsJumping()) -> (ReturnValue)
   n16 = Get bIsSurfaceSwimming()
   n17 = Get bIsSkyDiving()
   n18 = Get bIsFalling()
   > Set ASW_ClavicleFix(ASW_ClavicleFix:double="To Float (Boolean)"(InBool=OR Boolean!(A=@n15, B=@n16, C:bool=@n17, D:bool=@n18) -> (ReturnValue))) -> (Output_Get:double)
   n19 = Get bIsSkyDiving()
   n20 = OR Boolean!(
    A=Get bIsSprinting()
    B=Get bIsJumping()
    C=Get bIsSkyDiving()
    D=Get bIsPlayingMeleeAnim()
    E:bool=Get bIsMoving2D()
    F:bool=AND Boolean(A=Get bIsFalling(), B=NOT Boolean(A=@n19))
   ) -> (ReturnValue)
   > Set Alpha_Bag_Rot(Alpha_Bag_Rot:double=FInterpTo_Constant(Current=Get Alpha_Bag_Rot(), Target="To Float (Boolean)"(InBool=@n20), DeltaTime=Get DeltaTimeX(), InterpSpeed=2)) -> (Output_Get:double)
   n21 = Get bIsTacticalSprinting()
   n22 = Get bIsJumping()
   > Set Alpha_JumpingTacticalSprinting(Alpha_JumpingTacticalSprinting:double="To Float (Boolean)"(InBool=AND Boolean(A=@n21, B=@n22))) -> (Output_Get:double)
   n23 = Get bIsOnGround() -> (bIsOnGround:bool)
   n24 = Get bIsMoving2D()
   > Set Alpha_Moving2DGround(Alpha_Moving2DGround:double="To Float (Boolean)"(InBool=AND Boolean(A=@n24, B=@n23))) -> (Output_Get:double)
   > Set Alpha_Skydiving(Alpha_Skydiving:double="To Float (Boolean)"(InBool=Get bIsSkyDiving())) -> (Output_Get:double)
   > Set Alpha_Emote(Alpha_Emote:double="To Float (Boolean)"(InBool=Get bIsPlayingEmote())) -> (Output_Get:double)
   > n25 = Branch(Condition=Get bIsSkyDiving())
   > Set Alpha_SkydivingDyn(Alpha_SkydivingDyn:double=0.35) -> (Output_Get:double)
   Set Alpha_SkydivingDyn(execute=@n25.else, Alpha_SkydivingDyn=1)
   n26 = Branch(
    execute=@n2.then_4
    Condition=AND Boolean!(A="float > float"(A=@n10, B=20), B=Get bIsPlayingMeleeAnim(), C:bool=Get Selected_Pickaxe()) -> (ReturnValue)
   )
   > Set Alpha_DetectedArmBackL_InMelee(Alpha_DetectedArmBackL_InMelee:double=0.7) -> (Output_Get:double)
   Set Alpha_DetectedArmBackL_InMelee(execute=@n26.else, Alpha_DetectedArmBackL_InMelee=0)
   n27 = Event OnCharacterCustomizationLoaded()
   > n28 = Set Skel Mesh(Skel Mesh=@n27.CustomBaseMesh)
   Set RotHeadSkydive(
    execute=@n2.then_2
    RotHeadSkydive:double=float * float(
     A=GetAlphaFromReferenceTransformDelta(SkeletalMeshComponent=@n28, BoneName=spine_05, BaseName=neck_01, InputRange=(X=0,Y=25), TwistAxis="1,0,0")
     B=Get Alpha_Skydiving()
    ) {is=title K2Node_PromotableOperator}
   ) -> (Output_Get:double)
   Set Is Wearing Cape in Floating(
    execute=@n2.then_3
    Is Wearing Cape in Floating:bool=AND Boolean!(A=CheckForBackbling().bIsBackblingACape, B=Get bIsSkydiveFloating()) -> (ReturnValue)
   ) -> (Output_Get:bool)
   Event OnWeaponChanged()
   > n29 = Cast To FortPlayerAnimInstance(Object=GetAnimInstance(self=GetOwnerSkeletalMeshComponent()))
   > Set Selected_Pistol_Invert(Selected_Pistol_Invert:double=1) -> (Output_Get:double)
   > Set Selected_Pickaxe(Selected_Pickaxe:bool=false) -> (Output_Get:bool)
   > n30 = Switch on EFortWeaponCoreAnimation!(execute, Selection:EFortWeaponCoreAnimation Enum=Get WeaponCoreAnim(self=@n29)) -> (
    Melee:Exec
    Pistol:Exec
    Shotgun:Exec
    PaperBlueprint:Exec
    Rifle:Exec
    MeleeOneHand:Exec
    MachinePistol:Exec
    RocketLauncher:Exec
    GrenadeLauncher:Exec
    GoingCommando:Exec
    AssaultRifle:Exec
    TacticalShotgun:Exec
    SniperRifle:Exec
    TrapPlacement:Exec
    ShoulderLauncher:Exec
    AbilityDecoTool:Exec
    Crossbow:Exec
    C4:Exec
    RemoteControl:Exec
    DualWield:Exec
    AR_BullPup:Exec
    AR_ForwardGrip:Exec
    MedPackPaddles:Exec
    SMG_P90:Exec
    AR_DrumGun:Exec
    Consumable_Small:Exec
    Consumable_Large:Exec
    Balloon:Exec
    MountedTurret:Exec
    CreativeTool:Exec
    ExplosiveBow:Exec
    AshtonIndigo:Exec
    AshtonChicago:Exec
    MeleeDualWield:Exec
    Unarmed:Exec
   )
   > Set Selected_Pickaxe(Selected_Pickaxe=true)
   Set Selected_Pistol_Invert(execute=@n30.Pistol, Selected_Pistol_Invert=0)
   Event OnParachuteOpened()
   > Switch on EGliderType(Selection=Get PlayerGliderType() -> (PlayerGliderType=HangGlider))
   > Set HangGlidingAlpha(HangGlidingAlpha:double=1) -> (Output_Get:double)
   Event OnParachuteClosed()
   > Set HangGlidingAlpha(HangGlidingAlpha=0)
   "float > float"(A=@n10, B=-20)
  ]}
]
variables [
 Alpha_Bag_Rot {type=number}
 Alpha_Crouching {type=number}
 Alpha_DBNO {type=number}
 Alpha_DetectedArmBackL_InMelee {type=number}
 Alpha_DetectedArmBackR_InMelee_0 {type=number}
 Alpha_Emote {type=number}
 Alpha_Jumping {type=number}
 Alpha_JumpingTacticalSprinting {type=number}
 Alpha_Melee {type=number}
 Alpha_Moving2D {type=number}
 Alpha_Moving2DGround {type=number}
 Alpha_Skydiving {type=number}
 Alpha_SkydivingDyn {type=number}
 ASW_ClavicleFix {type=number}
 DeltaTimeX {type=number}
 Gravity Override {type=Vector}
 Gravity Override Params {type=GravityOverrideParamsStruct}
 HangGlidingAlpha {type=number}
 Is Wearing Cape in Floating {type=boolean}
 RBAN Params {
  type=RBANParamsStruct
  default {
   dBNO {simSpaceSettings {
     maxAngularAcceleration=1000
     maxAngularVelocity=50
     maxLinearAcceleration=300
     maxLinearVelocity=700
     velocityScaleZ=0
     worldAlpha=0.35
     externalLinearDragV=<0 0.5 0>
    }}
   dBNO_Carried {simSpaceSettings {
     maxAngularAcceleration=1000
     maxAngularVelocity=3
     maxLinearAcceleration=300
     maxLinearVelocity=700
     velocityScaleZ=0
     worldAlpha=0.4
     externalLinearDragV=<0 0.5 0>
    }}
   emoteOrMelee {simSpaceSettings {
     dampingAlpha=0.3
     maxAngularAcceleration=500
     maxAngularVelocity=50
     maxLinearAcceleration=150
     maxLinearVelocity=100
     velocityScaleZ=0.5
     worldAlpha=0.7
     externalLinearDragV=<1 1 1>
    }}
   front_End {simSpaceSettings {
     dampingAlpha=5
     maxAngularAcceleration=100
     maxAngularVelocity=5
     maxLinearAcceleration=300
     maxLinearVelocity=200
     velocityScaleZ=0.1
     worldAlpha=0.5
     externalLinearDragV=<0 0.5 0>
    }}
   onGround_Crouch_Moving {simSpaceSettings {
     dampingAlpha=0.1
     maxAngularAcceleration=1000
     maxAngularVelocity=1000
     maxLinearAcceleration=1000
     maxLinearVelocity=10000
     velocityScaleZ=20
     externalLinearDragV=<0 0.5 0>
    }}
   onGround_Falling {simSpaceSettings {
     maxAngularAcceleration=1000
     maxAngularVelocity=3
     maxLinearAcceleration=300
     maxLinearVelocity=700
     velocityScaleZ=0
     externalLinearDragV=<0 0.5 0>
    }}
   onGround_Moving {simSpaceSettings {
     dampingAlpha=0.3
     maxAngularAcceleration=500
     maxAngularVelocity=50
     maxLinearAcceleration=150
     maxLinearVelocity=100
     velocityScaleZ=0.5
     worldAlpha=0.7
     externalLinearDragV=<1 1 1>
    }}
   onGround_Standing {
    componentLinearAccScale=<0 0 0>
    componentLinearVelScale=<0 0 0>
    simSpaceSettings {
     maxAngularAcceleration=10
     maxAngularVelocity=10
     maxLinearAcceleration=10
     maxLinearVelocity=100
     velocityScaleZ=0.1
     externalLinearDragV=<0 0 0>
    }
   }
   rail_Grinding {simSpaceSettings {
     maxAngularAcceleration=1000
     maxAngularVelocity=3
     maxLinearAcceleration=300
     maxLinearVelocity=700
     velocityScaleZ=0
     worldAlpha=0.1
     externalLinearDragV=<0 0.5 0>
    }}
   riding_Moving {simSpaceSettings {
     maxAngularAcceleration=1000
     maxAngularVelocity=3
     maxLinearAcceleration=300
     maxLinearVelocity=700
     velocityScaleZ=0
     externalLinearDragV=<0 0.5 0>
    }}
   riding_Moving_Targeting {simSpaceSettings {
     maxAngularAcceleration=1000
     maxAngularVelocity=3
     maxLinearAcceleration=300
     maxLinearVelocity=700
     velocityScaleZ=0
     externalLinearDragV=<0 0.5 0>
    }}
   riding_Standing {simSpaceSettings {
     maxAngularAcceleration=1000
     maxAngularVelocity=3
     maxLinearAcceleration=300
     maxLinearVelocity=700
     velocityScaleZ=0
     externalLinearDragV=<0 0.5 0>
    }}
   skydive_DivingDown {
    componentLinearAccScale=<0 0 0>
    componentLinearVelScale=<0 0 0>
    simSpaceSettings {
     dampingAlpha=0
     maxAngularAcceleration=1000
     maxAngularVelocity=3
     maxLinearAcceleration=30
     maxLinearVelocity=700
     velocityScaleZ=0
     externalLinearDragV=<0 0.5 0>
    }
   }
   skydive_FloatingDown {
    componentLinearAccScale=<0 0 0>
    componentLinearVelScale=<0 0 0>
    simSpaceSettings {
     maxAngularAcceleration=1000
     maxAngularVelocity=3
     maxLinearAcceleration=300
     maxLinearVelocity=700
     velocityScaleZ=0
     worldAlpha=0.5
     externalLinearDragV=<0 0.5 0>
    }
   }
   skydive_GoingUp {simSpaceSettings {
     maxAngularAcceleration=1000
     maxAngularVelocity=3
     maxLinearAcceleration=300
     maxLinearVelocity=700
     velocityScaleZ=0
     externalLinearDragV=<0 0.5 0>
    }}
   skydive_Parachuting {
    componentLinearAccScale=<0 0 0>
    componentLinearVelScale=<0 0 0>
    simSpaceSettings {
     maxAngularAcceleration=500
     maxAngularVelocity=1
     maxLinearAcceleration=300
     maxLinearVelocity=100
     velocityScaleZ=0
     externalLinearDragV=<0 0.5 0>
    }
   }
   skydive_Surfing {simSpaceSettings {
     maxAngularAcceleration=1000
     maxAngularVelocity=3
     maxLinearAcceleration=300
     maxLinearVelocity=700
     velocityScaleZ=0
     externalLinearDragV=<0 0.5 0>
    }}
   swimming {simSpaceSettings {
     maxAngularAcceleration=1000
     maxAngularVelocity=3
     maxLinearAcceleration=300
     maxLinearVelocity=700
     velocityScaleZ=0
     externalLinearDragV=<0 0.5 0>
    }}
   tactical_Sprint {simSpaceSettings {
     maxAngularAcceleration=1000
     maxAngularVelocity=3
     maxLinearAcceleration=300
     maxLinearVelocity=700
     velocityScaleZ=0
     externalLinearDragV=<0 0.5 0>
    }}
   ziplining {simSpaceSettings {
     maxAngularAcceleration=1000
     maxAngularVelocity=3
     maxLinearAcceleration=300
     maxLinearVelocity=700
     velocityScaleZ=0
     externalLinearDragV=<0 0.5 0>
    }}
  }
 }
 RBAN Settings Out {
  type=RBANSettingStruct
  default {
   componentLinearAccScale=<0 0 0>
   componentLinearVelScale=<0 0 0>
   simSpaceSettings {
    dampingAlpha=1
    maxAngularAcceleration=1000
    maxAngularVelocity=3
    maxLinearAcceleration=300
    maxLinearVelocity=700
    velocityScaleZ=0
    worldAlpha=1
    externalAngularVelocity=<0 0 0>
    externalLinearDragV=<0 0.5 0>
    externalLinearVelocity=<0 0 0>
   }
  }
 }
 RotHeadSkydive {type=number}
 Selected_Pickaxe {type=boolean}
 Selected_Pistol_Invert {type=number}
 SimSpaceSettings {
  type=SimSpaceSettings
  default {
   dampingAlpha=1
   maxAngularAcceleration=10000
   maxAngularVelocity=10000
   maxLinearAcceleration=10000
   maxLinearVelocity=10000
   velocityScaleZ=1
   worldAlpha=0
   externalLinearDragV=<0 0 0>
  }
 }
 SimSpaceSettingsFEIdle {
  type=SimSpaceSettings
  default {
   dampingAlpha=5
   maxAngularAcceleration=100
   maxAngularVelocity=5
   maxLinearAcceleration=300
   maxLinearVelocity=200
   velocityScaleZ=0.1
   worldAlpha=0.2
   externalLinearDragV=<0 0.5 0>
  }
 }
 Skel Mesh {type=SkeletalMeshComponent}
]
